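# Physics asset for the left-side bow and skirt dynamic bones on the large male HealedScar body, named as the crouching variant.
# A pelvis capsule roots two branches: a three-box skirt chain and a main bow bone that fans into four capsule chains, A to D.
# Skirt, main bow and chain master joints rotate freely, while segments within each bow chain have tight swing limits and zero twist limit.
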PhysicsAsset /BRCosmetics/Characters/Player/Male/Large/Bodies/M_LRG_HealedScar/Meshes/M_LRG_HealedScar_Bow_isCrouching_Physics
bodies {
 dyn_bowA_l_01 [{shapeType=Capsule, length=5, radius=3, shapeName=dyn_bowA_l_01_capsule, center=<2 0 0>, rotation=<rot 90 0 0>}]
 dyn_bowA_l_02 [{shapeType=Capsule, length=5.5, radius=2.5, shapeName=dyn_bowA_l_02_capsule, center=<2 -1.6 0>, rotation=<rot 90 0 0>}]
 dyn_bowA_l_03 [{shapeType=Capsule, length=5, radius=3, shapeName=dyn_bowA_l_03_capsule, center=<2 -1.6 0>, rotation=<rot 90 0 0>}]
 dyn_bowA_l_MSTR [{shapeType=Capsule, length=0.5, radius=0.5, shapeName=dyn_bowA_l_MSTR_capsule}]
 dyn_bowB_l_01 [{shapeType=Capsule, length=4.5, radius=3.5, shapeName=dyn_bowB_l_01_capsule, center=<2 -1.9 0>, rotation=<rot 90 0 0>}]
 dyn_bowB_l_02 [{shapeType=Capsule, length=4.5, radius=3.5, shapeName=dyn_bowB_l_02_capsule, center=<2 -1.9 0>, rotation=<rot 90 0 0>}]
 dyn_bowB_l_03 [{shapeType=Capsule, length=4.5, radius=3.5, shapeName=dyn_bowB_l_03_capsule, center=<2 -1.9 0>, rotation=<rot 90 0 0>}]
 dyn_bowB_l_04 [{shapeType=Capsule, length=4.5, radius=3.5, shapeName=dyn_bowB_l_04_capsule, center=<2 -1.9 0>, rotation=<rot 90 0 0>}]
 dyn_bowB_l_MSTR [{shapeType=Capsule, length=0.5, radius=0.5, shapeName=dyn_bowB_l_MSTR_capsule}]
 dyn_bowC_l_01 [{shapeType=Capsule, length=5, radius=3, shapeName=dyn_bowC_l_01_capsule, center=<2 0 0>, rotation=<rot 90 0 0>}]
 dyn_bowC_l_02 [{shapeType=Capsule, length=5, radius=3, shapeName=dyn_bowC_l_02_capsule, center=<2 0 0>, rotation=<rot 90 0 0>}]
 dyn_bowC_l_MSTR [{shapeType=Capsule, length=0.5, radius=0.5, shapeName=dyn_bowC_l_MSTR_capsule}]
 dyn_bowD_l_01 [{shapeType=Capsule, length=6, radius=2, shapeName=dyn_bowD_l_01_capsule, center=<2 -1 0>, rotation=<rot 90 0 0>}]
 dyn_bowD_l_02 [{shapeType=Capsule, length=6, radius=2, shapeName=dyn_bowD_l_02_capsule, center=<2 -1 0>, rotation=<rot 90 0 0>}]
 dyn_bowD_l_MSTR [{shapeType=Capsule, length=0.5, radius=0.5, shapeName=dyn_bowD_l_MSTR_capsule}]
 dyn_bowMain_l_01 [{shapeType=Capsule, length=5, radius=3, shapeName=dyn_bowMain_l_01_capsule, center=<2 0 0>, rotation=<rot 90 0 0>}]
 dyn_bowMain_l_MSTR [{shapeType=Capsule, length=0.5, radius=0.5, shapeName=dyn_bowMain_l_MSTR_capsule}]
 dyn_skirtA_l_01 [{shapeType=Box, extentX=2, extentY=19, extentZ=37, shapeName=dyn_skirtA_l_01_box, center=<3.3081 2.5882 0>, rotation=<rot 4.1536 89.9999 0>}]
 dyn_skirtA_l_02 [{
   shapeType=Box
   extentX=2
   extentY=15
   extentZ=37
   shapeName=dyn_skirtA_l_02_box
   center=<6.8754 1.3591 -0.0481>
   rotation=<rot 4.0716 86.4925 -0.2492>
  }]
 dyn_skirtA_l_03 [{
   shapeType=Box
   extentX=2
   extentY=15
   extentZ=37
   shapeName=dyn_skirtA_l_03_box
   center=<6.8754 0.4236 -0.0682>
   rotation=<rot 4.1494 87.4029 -0.188>
  }]
 pelvis [{shapeType=Capsule, length=21.5, radius=11.7191, shapeName=pelvis_capsule}]
}
constraints [
 {
  bone1Name=dyn_bowMain_l_01
  bone2Name=dyn_bowMain_l_MSTR
  swing1LimitDegrees=45
  swing1Motion=Free
  swing2LimitDegrees=45
  swing2Motion=Free
  twistLimitDegrees=45
  twistMotion=Free
 }
 {
  bone1Name=dyn_bowA_l_MSTR
  bone2Name=dyn_bowMain_l_01
  swing1LimitDegrees=45
  swing1Motion=Free
  swing2LimitDegrees=45
  swing2Motion=Free
  twistLimitDegrees=45
  twistMotion=Free
 }
 {bone1Name=dyn_bowA_l_01, bone2Name=dyn_bowA_l_MSTR, swing1LimitDegrees=20, swing2LimitDegrees=10, twistLimitDegrees=0}
 {bone1Name=dyn_bowA_l_02, bone2Name=dyn_bowA_l_01, swing1LimitDegrees=20, swing2LimitDegrees=10, twistLimitDegrees=0}
 {bone1Name=dyn_bowA_l_03, bone2Name=dyn_bowA_l_02, swing1LimitDegrees=30, swing2LimitDegrees=10, twistLimitDegrees=0}
 {
  bone1Name=dyn_bowB_l_MSTR
  bone2Name=dyn_bowMain_l_01
  swing1LimitDegrees=45
  swing1Motion=Free
  swing2LimitDegrees=45
  swing2Motion=Free
  twistLimitDegrees=45
  twistMotion=Free
 }
 {bone1Name=dyn_bowB_l_01, bone2Name=dyn_bowB_l_MSTR, swing1LimitDegrees=20, swing2LimitDegrees=10, twistLimitDegrees=0}
 {bone1Name=dyn_bowB_l_02, bone2Name=dyn_bowB_l_01, swing1LimitDegrees=20, swing2LimitDegrees=10, twistLimitDegrees=0}
 {bone1Name=dyn_bowB_l_03, bone2Name=dyn_bowB_l_02, swing1LimitDegrees=30, swing2LimitDegrees=10, twistLimitDegrees=0}
 {bone1Name=dyn_bowB_l_04, bone2Name=dyn_bowB_l_03, swing1LimitDegrees=30, swing2LimitDegrees=10, twistLimitDegrees=0}
 {
  bone1Name=dyn_bowC_l_MSTR
  bone2Name=dyn_bowMain_l_01
  swing1LimitDegrees=45
  swing1Motion=Free
  swing2LimitDegrees=45
  swing2Motion=Free
  twistLimitDegrees=45
  twistMotion=Free
 }
 {bone1Name=dyn_bowC_l_01, bone2Name=dyn_bowC_l_MSTR, swing1LimitDegrees=20, swing2LimitDegrees=10, twistLimitDegrees=0}
 {bone1Name=dyn_bowC_l_02, bone2Name=dyn_bowC_l_01, swing1LimitDegrees=20, swing2LimitDegrees=10, twistLimitDegrees=0}
 {
  bone1Name=dyn_bowD_l_MSTR
  bone2Name=dyn_bowMain_l_01
  swing1LimitDegrees=45
  swing1Motion=Free
  swing2LimitDegrees=45
  swing2Motion=Free
  twistLimitDegrees=45
  twistMotion=Free
 }
 {bone1Name=dyn_bowD_l_01, bone2Name=dyn_bowD_l_MSTR, swing1LimitDegrees=20, swing2LimitDegrees=10, twistLimitDegrees=0}
 {bone1Name=dyn_bowD_l_02, bone2Name=dyn_bowD_l_01, swing1LimitDegrees=20, swing2LimitDegrees=10, twistLimitDegrees=0}
 {
  bone1Name=dyn_skirtA_l_02
  bone2Name=dyn_skirtA_l_01
  swing1LimitDegrees=45
  swing1Motion=Free
  swing2LimitDegrees=45
  swing2Motion=Free
  twistLimitDegrees=45
  twistMotion=Free
 }
 {
  bone1Name=dyn_skirtA_l_03
  bone2Name=dyn_skirtA_l_02
  swing1LimitDegrees=45
  swing1Motion=Free
  swing2LimitDegrees=45
  swing2Motion=Free
  twistLimitDegrees=45
  twistMotion=Free
 }
 {
  bone1Name=dyn_skirtA_l_01
  bone2Name=pelvis
  swing1LimitDegrees=45
  swing1Motion=Free
  swing2LimitDegrees=45
  swing2Motion=Free
  twistLimitDegrees=45
  twistMotion=Free
 }
 {
  bone1Name=dyn_bowMain_l_MSTR
  bone2Name=pelvis
  swing1LimitDegrees=45
  swing1Motion=Free
  swing2LimitDegrees=45
  swing2Motion=Free
  twistLimitDegrees=45
  twistMotion=Free
 }
]
