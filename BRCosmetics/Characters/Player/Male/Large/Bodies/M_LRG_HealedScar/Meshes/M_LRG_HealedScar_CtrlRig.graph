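# Control Rig for the M_LRG_HealedScar large male player body, correcting deformation and secondary dyn_ bones.
# The shared DeformRig runs first, then shoulder armor correctives (one skipped while skydiving) and weighted constraints for rope, scarf and shirt.
# Thigh spherical pose readers feed variables that, with movement blends and thigh constraints, shape the skirt, which the bows follow.
# Spherical collision finally constrains neck bones around the head.
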ControlRigBlueprint /BRCosmetics/Characters/Player/Male/Large/Bodies/M_LRG_HealedScar/Meshes/M_LRG_HealedScar_CtrlRig
graphs [
 RigVMFunctionLibrary {links [], nodes []}
 RigVMFunction_Sequence_1_ContainedGraph {nodes [
   n1 = RigVMFunctionEntryNode()
   RigVMFunction_Sequence = &RigVMFunction_Sequence(ExecuteContext=@n1.ExecuteContext)
   n2 = &RigVMFunction_Sequence(ExecuteContext=@RigVMFunction_Sequence.B)
   n3 = &RigVMFunction_Sequence(ExecuteContext=@n2.B)
   RigVMFunctionReturnNode(A=@RigVMFunction_Sequence.A, B=@n2.A, C=@n3.A, D=@n3.B)
  ]}
 RigVMModel {nodes [
   RigUnit_BeginExecution = &RigUnit_BeginExecution()
   RigVMFunction_Sequence = RigVMAggregateNode(ExecuteContext=@RigUnit_BeginExecution.ExecutePin)
   Spherical Collision = RigVMFunctionReferenceNode(
    Debug=false
    Max Radius=20
    Min Radius=19
    Source=(Type=Bone,Name="dyn_scarNeck_c_01")
    Source Offset=(X=0,Y=0,Z=0)
    Target=(Type=Bone,Name="head")
    Target Offset=(X=4,Y=0,Z=0)
    Weight=1
    ExecuteContext=@RigVMFunction_Sequence.D
   )
   > Spherical Collision_1 = RigVMFunctionReferenceNode(
    Debug=false
    Max Radius=15
    Min Radius=15
    Source=(Type=Bone,Name="dyn_scarNeck_l_01")
    Source Offset=(X=0,Y=0,Z=0)
    Target=(Type=Bone,Name="head")
    Target Offset=(X=4,Y=0,Z=0)
    Weight=1
   )
   > Spherical Collision_1_1 = RigVMFunctionReferenceNode(
    Debug=false
    Max Radius=20
    Min Radius=19
    Source=(Type=Bone,Name="dyn_scarNeck_r_01")
    Source Offset=(X=0,Y=0,Z=0)
    Target=(Type=Bone,Name="head")
    Target Offset=(X=4,Y=0,Z=0)
    Weight=1
   )
   FN_DeformRig_FNC = RigVMFunctionReferenceNode(
    DeformRig_ Curve="(deformRig_l_upperarm_up_alpha_72_#1=true,deformRig_r_upperarm_up_alpha_73_#2=False,deformRig_l_upperarm_fr_alpha_74_#3=true,deformRig_r_upperarm_fr_alpha_75_CB51B80C ...(truncated)"
    DeformRig_Bool="(upperarm_1_#4=True,elbow_3_#5=True,wrist_7_#6=True,groin_8_#7=false,glute_17_#8=True,knee_9_40C408324CA26 ...(truncated)"
    DeformRig_Debug=(upperarm_debug_18_#4=False,wrist_debug_20_#6=False,lat_debug_26_#9=False,groinGlute_debug_27_#7=False)
    DeformRig_Value="(upperarm_up_18_#10=(Rotation=(X=0,Y=0,Z=0,W=1),Translation=(X=0,Y=0,Z=8),Scale3D=(X=0.7,Y=1,Z=1)),upperarm_fwd_fr_19_#11=(Rotation=(X=0,Y=0,Z=0,W=1),Translation=(X=0,Y=10,Z=0),Scale3D=(X=0.5,Y=1,Z= ...(truncated)"
    ExecuteContext=@RigVMFunction_Sequence.A
   )
   > &RigUnit_ModifyTransforms(
    ItemToModify="((Item=(Type=Bone,Name=\"dyn_shoulderPad_l_01\"),Transform=(Rotation=(X=0,Y=0,Z=0,W=1),Translation=(X=0,Y=0,Z=0),Scale3D=(X=1,Y=1,Z=1))),(Item=(Type=Bone,Name=\"dyn_shoulderPlateA_l_01\"),Transform=(Rotation=(X=0,Y=0,Z=0,W=1),Translation=(X=0,Y=0,Z=0),Scale3D= ...(truncated)"
    WeightMinimum=0
    Weight=RigVMDispatchNode(BoneName=root, Default=0, Name=deformRig_l_upperarm_fr_alpha) {template="DISPATCH_RigDispatch_GetAnimAttribute(in Name,in BoneName,in Default,out Value,out Found)"}.Value
   )
   > n1 = &RigUnit_ModifyTransforms(
    ItemToModify="((Item=(Type=Bone,Name=\"dyn_shoulderPad_l_01\"),Transform=(Rotation=(X=0,Y=0,Z=0,W=1),Translation=(X=1,Y=2,Z=1),Scale3D=(X=1,Y=1,Z=1))),(Item=(Type=Bone,Name=\"dyn_shoulderPlateA_l_01\"),Transform=(Rotation=(X=0,Y=0,Z=0,W=1),Translation=(X=0,Y=0,Z=0),Scale3D= ...(truncated)"
    WeightMinimum=0
    Weight=RigVMDispatchNode(BoneName=root, Default=0, Name=deformRig_l_upperarm_up_alpha) {template="DISPATCH_RigDispatch_GetAnimAttribute(in Name,in BoneName,in Default,out Value,out Found)"}.Value
   )
   RigVMFunction_ControlFlowBranch = &RigVMFunction_ControlFlowBranch(ExecuteContext=@n1.ExecutePin, Condition=RigVMVariableNode() {variable=is_not_skydiving}.Value)
   &RigUnit_ParentConstraint(
    AdvancedSettings=(RotationOrderForFilter=XZY)
    Child=(Type=Bone,Name="dyn_bentoRopeAim_MSTR")
    Parents="((Item=(Type=Bone,Name=\"deform_glute_r\"),Weight=0.0002),(Item=(Type=Bone,Name=\"dyn_skirtD_l_02\"),Weight=0.0004),(Item=(Type=Bone,Name=\"dyn_skirtD_l_01\"),Weight=0.003),(Item=(Type=Bone,Name=\"dyn_skirtD_r_01\"),Weight=0.0112),(Item=(Type=Bone,Name=\"spine_03\") ...(truncated)"
    Weight=1
    ExecutePin=@RigVMFunction_ControlFlowBranch.Completed
   )
   > &RigUnit_ParentConstraint(
    AdvancedSettings=(RotationOrderForFilter=XZY)
    Child=(Type=Bone,Name="dyn_scarfFlap_l_MSTR")
    Parents="((Item=(Type=Bone,Name=\"head\"),Weight=0.0033),(Item=(Type=Bone,Name=\"neck_01\"),Weight=0.0216),(Item=(Type=Bone,Name=\"neck_02\"),Weight=0.0223),(Item=(Type=Bone,Name=\"deform_upperarm_up_l\"),Weight=0.0251),(Item=(Type=Bone,Name=\"spine_04\"),Weight=0.0372),(Ite ...(truncated)"
    Weight=1
   )
   > &RigUnit_ParentConstraint(
    AdvancedSettings=(RotationOrderForFilter=XZY)
    Child=(Type=Bone,Name="dyn_shirtJiggle_c_01")
    Parents="((Item=(Type=Bone,Name=\"clavicle_l\"),Weight=0.0002),(Item=(Type=Bone,Name=\"spine_04\"),Weight=0.0614),(Item=(Type=Bone,Name=\"spine_01\"),Weight=0.1713),(Item=(Type=Bone,Name=\"pelvis\"),Weight=0.1818),(Item=(Type=Bone,Name=\"spine_02\"),Weight=0.2901),(Item=(Typ ...(truncated)"
    Weight=1
   )
   n2 = &RigUnit_SphericalPoseReader(
    ActiveRegionScaleFactors=(PositiveWidth=2,NegativeWidth=6,PositiveHeight=6,NegativeHeight=6)
    ActiveRegionSize=0.04
    Debug=(bDrawDebug=true,DebugScale=25,DebugSegments=20,DebugThickness=0.25)
    DriverAxis=(X=1,Y=0,Z=0)
    DriverItem=(Type=Bone,Name="upperarm_l")
    FalloffRegionScaleFactors=(PositiveWidth=1,NegativeWidth=1,PositiveHeight=1,NegativeHeight=1)
    FalloffSize=0.38
    OptionalParentItem=(Type=Bone)
    RotationOffset=(X=0,Y=-40,Z=-20)
    ExecutePin=@RigVMFunction_ControlFlowBranch.True
   )
   > &RigUnit_ModifyTransforms(
    ItemToModify=((Item=(Type=Bone,Name="dyn_shoulderPad_l_01"),Transform=(Rotation=(X=0.0868,Y=0.0076,Z=0.0868,W=0.9924),Translation=(X=8,Y=0,Z=0),Scale3D=(X=1,Y=1,Z=1))))
    Weight=@n2.OutputParam
   )
   n3 = &RigUnit_SphericalPoseReader(
    ActiveRegionScaleFactors=(PositiveWidth=1,NegativeWidth=0.5,PositiveHeight=1,NegativeHeight=1)
    ActiveRegionSize=0.1
    Debug=(bDrawDebug=true,DebugScale=25,DebugSegments=20,DebugThickness=0.25)
    DriverAxis=(X=-1,Y=0,Z=0)
    DriverItem=(Type=Bone,Name="thigh_l")
    FalloffRegionScaleFactors=(PositiveWidth=1,NegativeWidth=0.6,PositiveHeight=1,NegativeHeight=1)
    FalloffSize=0.45
    OptionalParentItem=(Type=Bone)
    RotationOffset=(X=120,Y=-40,Z=0)
    ExecutePin=@RigVMFunction_Sequence.B
   )
   > RigVMVariableNode(Value=@n3.OutputParam) {variable=thigh_l_up}
   > n4 = &RigUnit_SphericalPoseReader(
    ActiveRegionScaleFactors=(PositiveWidth=4,NegativeWidth=0.5,PositiveHeight=4.5,NegativeHeight=4.5)
    ActiveRegionSize=0.1
    Debug=(bDrawDebug=false,DebugScale=25,DebugSegments=20,DebugThickness=0.25)
    DriverAxis=(X=1,Y=0,Z=0)
    DriverItem=(Type=Bone,Name="thigh_r")
    FalloffRegionScaleFactors=(PositiveWidth=1,NegativeWidth=0.7,PositiveHeight=5,NegativeHeight=5)
    FalloffSize=0.45
    FlipHeightScaling=false
    FlipWidthScaling=true
    OptionalParentItem=(Type=Bone)
    RotationOffset=(X=-60,Y=-5,Z=5)
   )
   > RigVMVariableNode(Value=@n4.OutputParam) {variable=thigh_r_up}
   > n5 = &RigUnit_SphericalPoseReader(
    ActiveRegionScaleFactors=(PositiveWidth=1,NegativeWidth=0.5,PositiveHeight=1,NegativeHeight=1)
    ActiveRegionSize=0.1
    Debug=(bDrawDebug=true,DebugScale=25,DebugSegments=20,DebugThickness=0.25)
    DriverAxis=(X=1,Y=0,Z=0)
    DriverItem=(Type=Bone,Name="thigh_r")
    FalloffRegionScaleFactors=(PositiveWidth=0.5,NegativeWidth=0.5,PositiveHeight=0.5,NegativeHeight=0.5)
    FalloffSize=0.45
    FlipHeightScaling=false
    FlipWidthScaling=true
    OptionalParentItem=(Type=Bone)
    RotationOffset=(X=-30,Y=-30,Z=5)
   )
   > RigVMVariableNode(Value=@n5.OutputParam) {variable=thigh_r_out}
   &RigUnit_ParentConstraint(
    AdvancedSettings=(InterpolationType=Average,RotationOrderForFilter=XZY)
    Child=(Type=Bone,Name="dyn_skirtA_l_MSTR")
    Filter=(TranslationFilter=(bX=false,bY=false,bZ=false),RotationFilter=(bX=True,bY=True,bZ=True),ScaleFilter=(bX=True,bY=True,bZ=True))
    Parents=((Item=(Type=Bone,Name="thigh_twist_01_l"),Weight=1))
    Weight=1
    ExecutePin=@RigVMFunction_Sequence.C
   )
   > &RigUnit_ParentConstraint(
    AdvancedSettings=(InterpolationType=Average,RotationOrderForFilter=XZY)
    Child=(Type=Bone,Name="dyn_skirtA_r_MSTR")
    Filter=(TranslationFilter=(bX=false,bY=false,bZ=false),RotationFilter=(bX=True,bY=True,bZ=True),ScaleFilter=(bX=True,bY=True,bZ=True))
    Parents=((Item=(Type=Bone,Name="thigh_twist_01_r"),Weight=1))
    Weight=1
   )
   RigVMFunction_AnimEvalRichCurve = &RigVMFunction_AnimEvalRichCurve(
    Curve=(EditorCurveData=(Keys=((InterpMode=RCIM_Cubic,TangentMode=RCTM_User,Value=1),(InterpMode=RCIM_Cubic,Time=1))))
    SourceMinimum=0
    TargetMinimum=0
    Value=RigVMVariableNode() {variable=is_player_moving}.Value
   )
   > &RigUnit_ModifyTransforms(
    ItemToModify="((Item=(Type=Bone,Name=\"dyn_skirtA_l_MSTR\"),Transform=(Rotation=(X=0,Y=0,Z=-0.0872,W=0.9962),Translation=(X=0,Y=-2,Z=0),Scale3D=(X=1,Y=1,Z=1))),(Item=(Type=Bone,Name=\"dyn_skirtB_l_MSTR\"),Transform=(Rotation=(X=0,Y=0,Z=-0.0698,W=0.9976),Translation=(X=0,Y=2 ...(truncated)"
    WeightMinimum=0
    Weight=@RigVMFunction_AnimEvalRichCurve.Result
   )
   AlphaInterp = &RigVMFunction_AlphaInterp(InterpSpeedDecreasing=3, InterpSpeedIncreasing=3, Scale=1, bClampResult=False, Value=RigVMVariableNode() {variable=is_player_moving}.Value)
   > &RigUnit_ModifyTransforms(
    ItemToModify="((Item=(Type=Bone,Name=\"dyn_skirtA_l_MSTR\"),Transform=(Rotation=(X=0,Y=0,Z=-0.0872,W=0.9962),Translation=(X=0,Y=-2,Z=0),Scale3D=(X=1,Y=1,Z=1))),(Item=(Type=Bone,Name=\"dyn_skirtB_l_MSTR\"),Transform=(Rotation=(X=0,Y=0,Z=-0.0872,W=0.9962),Translation=(X=0,Y=2 ...(truncated)"
    WeightMinimum=0
    Weight=@AlphaInterp.Result
   )
   > &RigUnit_ParentConstraint(
    AdvancedSettings=(InterpolationType=Average,RotationOrderForFilter=XZY)
    Child=(Type=Bone,Name="dyn_skirtB_l_MSTR")
    Parents=((Item=(Type=Bone,Name="thigh_l"),Weight=1))
    Weight=0.8
   )
   > &RigUnit_ParentConstraint(
    AdvancedSettings=(InterpolationType=Average,RotationOrderForFilter=XZY)
    Child=(Type=Bone,Name="dyn_skirtB_r_MSTR")
    Parents=((Item=(Type=Bone,Name="thigh_r"),Weight=1))
    Weight=0.8
   )
   > &RigUnit_ParentConstraint(
    AdvancedSettings=(InterpolationType=Average,RotationOrderForFilter=XZY)
    Child=(Type=Bone,Name="dyn_skirtC_l_MSTR")
    Parents=((Item=(Type=Bone,Name="thigh_l"),Weight=1))
    Weight=0.3
   )
   > &RigUnit_ParentConstraint(
    AdvancedSettings=(InterpolationType=Average,RotationOrderForFilter=XZY)
    Child=(Type=Bone,Name="dyn_skirtC_r_MSTR")
    Parents=((Item=(Type=Bone,Name="thigh_r"),Weight=1))
    Weight=0.5
   )
   > &RigUnit_ModifyTransforms(
    ItemToModify="((Item=(Type=Bone,Name=\"dyn_skirtA_r_MSTR\"),Transform=(Rotation=(X=0,Y=0,Z=-0.0872,W=0.9962),Translation=(X=0,Y=-5,Z=0),Scale3D=(X=1,Y=1,Z=1))),(Item=(Type=Bone,Name=\"dyn_skirtB_r_MSTR\"),Transform=(Rotation=(X=0,Y=0,Z=-0.1736,W=0.9848),Translation=(X=0,Y=0 ...(truncated)"
    WeightMinimum=0
    Weight=RigVMVariableNode() {variable=thigh_r_out}.Value
   )
   > &RigUnit_ModifyTransforms(
    ItemToModify="((Item=(Type=Bone,Name=\"dyn_skirtA_l_MSTR\"),Transform=(Rotation=(X=0.05,Y=0.5714,Z=-0.0714,W=0.816),Translation=(X=-3,Y=-5,Z=0),Scale3D=(X=1,Y=1,Z=1))),(Item=(Type=Bone,Name=\"dyn_skirtB_l_MSTR\"),Transform=(Rotation=(X=-0.0298,Y=0.3407,Z=0.0819,W=0.9361),Tr ...(truncated)"
    WeightMaximum=0.8
    WeightMinimum=0
    Weight=RigVMVariableNode() {variable=thigh_l_up}.Value
   )
   > &RigUnit_ModifyTransforms(
    ItemToModify="((Item=(Type=Bone,Name=\"dyn_skirtA_r_MSTR\"),Transform=(Rotation=(X=0.0552,Y=-0.419,Z=0.1183,W=0.8986),Translation=(X=-3,Y=-3,Z=0),Scale3D=(X=1,Y=1,Z=1))),(Item=(Type=Bone,Name=\"dyn_skirtB_r_MSTR\"),Transform=(Rotation=(X=-0.0226,Y=-0.2578,Z=-0.0842,W=0.9623 ...(truncated)"
    WeightMinimum=0
    Weight=RigVMVariableNode() {variable=thigh_r_up}.Value
   )
   > &RigUnit_ModifyTransforms(
    ItemToModify=((Item=(Type=Bone,Name="dyn_skirtA_r_MSTR"),Transform=(Rotation=(X=0,Y=0,Z=-0.0872,W=0.9962),Translation=(X=0,Y=0,Z=0),Scale3D=(X=1,Y=1,Z=1))))
    Weight=1
    WeightMinimum=0
   )
   > &RigUnit_ParentConstraint(
    AdvancedSettings=(InterpolationType=Average,RotationOrderForFilter=XZY)
    Child=(Type=Bone,Name="dyn_bowA_l_MSTR")
    Filter=(TranslationFilter=(bX=true,bY=true,bZ=true),RotationFilter=(bX=True,bY=true,bZ=true),ScaleFilter=(bX=True,bY=True,bZ=True))
    Parents=((Item=(Type=Bone,Name="dyn_skirtA_l_MSTR"),Weight=1))
    Weight=1
   )
   > &RigUnit_ParentConstraint(
    AdvancedSettings=(InterpolationType=Average,RotationOrderForFilter=XZY)
    Child=(Type=Bone,Name="dyn_bowB_l_MSTR")
    Filter=(TranslationFilter=(bX=true,bY=true,bZ=true),RotationFilter=(bX=True,bY=true,bZ=True),ScaleFilter=(bX=True,bY=True,bZ=True))
    Parents=((Item=(Type=Bone,Name="dyn_skirtA_l_MSTR"),Weight=1))
    Weight=1
   )
   > &RigUnit_ParentConstraint(
    AdvancedSettings=(InterpolationType=Average,RotationOrderForFilter=XZY)
    Child=(Type=Bone,Name="dyn_bowC_l_MSTR")
    Filter=(TranslationFilter=(bX=false,bY=false,bZ=false),RotationFilter=(bX=True,bY=true,bZ=True),ScaleFilter=(bX=True,bY=True,bZ=True))
    Parents=((Item=(Type=Bone,Name="dyn_skirtA_l_01"),Weight=1))
    Weight=1
   )
   > &RigUnit_ParentConstraint(
    AdvancedSettings=(InterpolationType=Average,RotationOrderForFilter=XZY)
    Child=(Type=Bone,Name="dyn_bowD_l_MSTR")
    Filter=(TranslationFilter=(bX=false,bY=false,bZ=false),RotationFilter=(bX=True,bY=false,bZ=True),ScaleFilter=(bX=True,bY=True,bZ=True))
    Parents=((Item=(Type=Bone,Name="dyn_skirtA_l_01"),Weight=1))
    Weight=1
   )
   &RigUnit_ModifyTransforms(
    ExecutePin=()
    ItemToModify="((Item=(Type=Bone,Name=\"dyn_skirtA_l_MSTR\"),Transform=(Rotation=(X=0.4451,Y=0.3653,Z=-0.598,W=0.5575),Translation=(X=-2,Y=2,Z=0),Scale3D=(X=1,Y=1,Z=1))),(Item=(Type=Bone,Name=\"dyn_skirtB_l_MSTR\"),Transform=(Rotation=(X=0.36,Y=0.1022,Z=-0.2192,W=0.901),Tran ...(truncated)"
    Weight=1
    WeightMaximum=0.8
    WeightMinimum=0
   )
   &RigUnit_ModifyTransforms(
    ExecutePin=()
    ItemToModify="((Item=(Type=Bone,Name=\"None\"),Transform=(Rotation=(X=0,Y=0,Z=0,W=1),Translation=(X=0,Y=0,Z=0),Scale3D=(X=1,Y=1,Z=1))),(Item=(Type=Bone,Name=\"dyn_skirtA_r_MSTR\"),Transform=(Rotation=(X=-0.2706,Y=-0.2706,Z=-0.6533,W=0.6533),Translation=(X=-4,Y=0,Z=3),Scale3 ...(truncated)"
    Weight=1
    WeightMinimum=0
   )
  ]}
]
hierarchy [
 root {loc=<0 0 0>, rot=<0 0 0 1>}
 pelvis {parent=root, loc=<0 -0.7321 92.4218>, rot=<0.0127 0.707 -0.0127 -0.707>}
 spine_01 {parent=pelvis, loc=<5.6002 -0.5128 0>, rot=<0 0 0.034 -0.9994>}
 spine_02 {parent=spine_01, loc=<8.0756 0.0139 0>, rot=<0 0 0.0238 -0.9997>}
 spine_03 {parent=spine_02, loc=<7.0395 0 0>, rot=<0 0 0.0149 -0.9999>}
 spine_04 {parent=spine_03, loc=<8.1023 0 0>, rot=<0 0 -0.0132 -0.9999>}
 spine_05 {parent=spine_04, loc=<10.0141 0 0>, rot=<0 0 -0.0015 -1>}
 attach_backpack {parent=spine_05, loc=<-0.123 -15.1423 0>, rot=<-0.0093 -0.707 0.0093 -0.707>}
 clavicle_l {parent=spine_05, loc=<2.5856 -0.2866 -4.3777>, rot=<0.0465 -0.7289 -0.009 -0.6829>}
 upperarm_l {parent=clavicle_l, loc=<18.2264 0.1342 -0.001>, rot=<0.0253 -0.2693 0.0013 -0.9627>}
 lowerarm_l {parent=upperarm_l, loc=<29.0071 0.1686 -1.239>, rot=<0.0391 -0.0099 -0.2453 -0.9686>}
 hand_l {parent=lowerarm_l, loc=<25.8607 0.222 1.0274>, rot=<-0.7193 -0.0988 0.1193 -0.6773>}
 thumb_01_l {parent=hand_l, loc=<1.3118 -0.3191 -2.459>, rot=<0.6546 -0.4042 0.0664 -0.6354>}
 thumb_02_l {parent=thumb_01_l, loc=<5.9336 -0.0448 0.0401>, rot=<-0.0037 0.0213 0.298 -0.9543>}
 thumb_03_l {parent=thumb_02_l, loc=<2.2223 0.3682 -0.0907>, rot=<0.0118 -0.013 0.0066 -0.9998>}
 index_metacarpal_l {parent=hand_l, loc=<1.6535 0.0183 -2.96>, rot=<0.0158 -0.0124 -0.0173 -0.9996>}
 index_01_l {parent=index_metacarpal_l, loc=<8.3008 0.0357 -0.0524>, rot=<-0.0044 0.0485 0.0924 -0.9945>}
 index_02_l {parent=index_01_l, loc=<3.3587 -0.3004 0.0452>, rot=<-0.002 0.0027 0.3225 -0.9466>}
 index_03_l {parent=index_02_l, loc=<2.5578 -0.123 -0.0252>, rot=<-0.002 0.0067 0.2558 -0.9667>}
 middle_metacarpal_l {parent=hand_l, loc=<2.1856 -0.1725 -0.5984>, rot=<0.006 0.0168 -0.0329 -0.9993>}
 middle_01_l {parent=middle_metacarpal_l, loc=<7.3275 -0.2447 0.0129>, rot=<-0.0208 0.0042 0.0858 -0.9961>}
 middle_02_l {parent=middle_01_l, loc=<4.0123 0.1401 0.1175>, rot=<-0.0218 0.0258 0.3512 -0.9357>}
 middle_03_l {parent=middle_02_l, loc=<2.6077 0.0534 0.1373>, rot=<0.0082 0.0188 0.1708 -0.9851>}
 ring_metacarpal_l {parent=hand_l, loc=<2.2074 -0.3625 1.3638>, rot=<-0.0284 0.0722 -0.0458 -0.9959>}
 ring_01_l {parent=ring_metacarpal_l, loc=<7.0295 -0.6197 -0.1446>, rot=<-0.0489 -0.0252 0.1892 -0.9804>}
 ring_02_l {parent=ring_01_l, loc=<3.6174 0.4647 -0.1027>, rot=<0 0 0.2806 -0.9598>}
 ring_03_l {parent=ring_02_l, loc=<2.4152 0.114 0.0199>, rot=<0 0 0.1189 -0.9929>}
 pinky_metacarpal_l {parent=hand_l, loc=<1.7616 -0.9753 3.3814>, rot=<-0.0397 0.0814 -0.069 -0.9935>}
 pinky_01_l {parent=pinky_metacarpal_l, loc=<7.1709 -0.4831 -0.1094>, rot=<-0.0654 -0.0035 0.2242 -0.9723>}
 pinky_02_l {parent=pinky_01_l, loc=<2.8274 0.0893 -0.1043>, rot=<0.0037 0.0011 0.2854 -0.9584>}
 pinky_03_l {parent=pinky_02_l, loc=<2.2046 0.0307 -0.0078>, rot=<-0.0002 0.0098 0.1445 -0.9895>}
 deform_wrist_dn_base_l {parent=hand_l, loc=<0 -3.4173 0>, rot=<0 0 0 -1>}
 deform_wrist_dn_l {parent=deform_wrist_dn_base_l, loc=<0 0 0>, rot=<0 0 0 1>}
 deform_wrist_up_base_l {parent=hand_l, loc=<0 2.7805 0>, rot=<0 0 0 -1>}
 deform_wrist_up_l {parent=deform_wrist_up_base_l, loc=<0 0 0>, rot=<0 0 0 1>}
 lowerarm_twist_02_l {parent=lowerarm_l, loc=<9.6199 0 0>, rot=<0 0 0 -1>}
 lowerarm_twist_01_l {parent=lowerarm_l, loc=<17.8549 0 0>, rot=<0 0 0 -1>}
 deform_elbow_in_base_l {parent=lowerarm_l, loc=<0.2112 5.645 0.8274>, rot=<0 0 0 -1>}
 deform_elbow_in_l {parent=deform_elbow_in_base_l, loc=<0 0 0>, rot=<0 0 0 1>}
 deform_elbow_out_base_l {parent=lowerarm_l, loc=<0 -7.5129 0>, rot=<0 0 0 -1>}
 deform_elbow_out_l {parent=deform_elbow_out_base_l, loc=<0 0 0>, rot=<0 0 0 1>}
 upperarm_twist_01_l {parent=upperarm_l, loc=<7.9796 0.0843 -0.6195>, rot=<0 0 0 -1>}
 deform_upperarm_bk_base_l {parent=upperarm_twist_01_l, loc=<-6.8983 -13.3826 -1.3258>, rot=<0 0 0 1>}
 deform_upperarm_bk_l {parent=deform_upperarm_bk_base_l, loc=<0 0 0>, rot=<0 0 0 1>}
 deform_upperarm_fr_base_l {parent=upperarm_twist_01_l, loc=<-8.3487 10.7117 -2.1455>, rot=<0 0 0 1>}
 deform_upperarm_fr_l {parent=deform_upperarm_fr_base_l, loc=<0 0 0>, rot=<0 0 0 1>}
 deform_upperarm_up_base_l {parent=upperarm_twist_01_l, loc=<-11.2012 -1.2234 8.7784>, rot=<0 0 0 1>}
 deform_upperarm_up_l {parent=deform_upperarm_up_base_l, loc=<0 0 0>, rot=<0 0 0 1>}
 upperarm_twist_02_l {parent=upperarm_l, loc=<19.5059 0.0843 -0.6195>, rot=<0 0 0 -1>}
 deform_bicep_base_l {parent=upperarm_twist_02_l, loc=<0 8.0494 0>, rot=<0 0 0 1>}
 deform_bicep_l {parent=deform_bicep_base_l, loc=<0 0 0>, rot=<0 0 0 1>}
 dyn_shoulderPad_l_01 {parent=upperarm_l, loc=<0 0 0>, rot=<0 0 0 1>}
 dyn_shoulderPlateA_l_01 {parent=dyn_shoulderPad_l_01, loc=<4.3135 -1.9401 9.7314>, rot=<0.6753 -0.1389 -0.1559 -0.7074>}
 dyn_shoulderPlateB_l_01 {parent=dyn_shoulderPlateA_l_01, loc=<5.3581 -3.9313 0.0518>, rot=<0.0012 0.0008 0.067 -0.9978>}
 dyn_shoulderPlateC_l_01 {parent=dyn_shoulderPlateB_l_01, loc=<10.4541 -0.4355 0.003>, rot=<-0.0006 -0.0004 -0.0298 -0.9996>}
 neck_01 {parent=spine_05, loc=<10.815 1.9196 0>, rot=<0 0 -0.1864 -0.9825>}
 neck_02 {parent=neck_01, loc=<4.7347 0.3045 0>, rot=<0 0 0.0363 -0.9993>}
 head {parent=neck_02, loc=<5.7632 0.5552 0>, rot=<0 0 0.115 -0.9934>}
 attach_cape {parent=spine_05, loc=<9.1509 -13.3783 0>, rot=<0.0616 -0.7044 -0.0616 -0.7044>}
 clavicle_r {parent=spine_05, loc=<2.5855 -0.2866 4.3777>, rot=<0.7289 0.0465 -0.6829 0.009>}
 upperarm_r {parent=clavicle_r, loc=<-18.2264 -0.1342 0.001>, rot=<0.0253 -0.2693 0.0013 -0.9627>}
 lowerarm_r {parent=upperarm_r, loc=<-29.0071 -0.1686 1.239>, rot=<0.0391 -0.0099 -0.2453 -0.9686>}
 hand_r {parent=lowerarm_r, loc=<-25.8607 -0.222 -1.0274>, rot=<-0.7193 -0.0988 0.1193 -0.6773>}
 thumb_01_r {parent=hand_r, loc=<-1.3118 0.3191 2.459>, rot=<0.6546 -0.4042 0.0664 -0.6354>}
 thumb_02_r {parent=thumb_01_r, loc=<-5.9336 0.0448 -0.0401>, rot=<-0.0037 0.0213 0.298 -0.9543>}
 thumb_03_r {parent=thumb_02_r, loc=<-2.2223 -0.3682 0.0907>, rot=<0.0118 -0.013 0.0066 -0.9998>}
 index_metacarpal_r {parent=hand_r, loc=<-1.6535 -0.0183 2.96>, rot=<0.0158 -0.0124 -0.0173 -0.9996>}
 index_01_r {parent=index_metacarpal_r, loc=<-8.3008 -0.0357 0.0524>, rot=<-0.0044 0.0485 0.0924 -0.9945>}
 index_02_r {parent=index_01_r, loc=<-3.3587 0.3004 -0.0452>, rot=<-0.002 0.0027 0.3225 -0.9466>}
 index_03_r {parent=index_02_r, loc=<-2.5578 0.123 0.0252>, rot=<-0.002 0.0067 0.2558 -0.9667>}
 middle_metacarpal_r {parent=hand_r, loc=<-2.1856 0.1725 0.5984>, rot=<0.006 0.0168 -0.0329 -0.9993>}
 middle_01_r {parent=middle_metacarpal_r, loc=<-7.3275 0.2447 -0.0129>, rot=<-0.0208 0.0042 0.0858 -0.9961>}
 middle_02_r {parent=middle_01_r, loc=<-4.0123 -0.1401 -0.1175>, rot=<-0.0218 0.0258 0.3512 -0.9357>}
 middle_03_r {parent=middle_02_r, loc=<-2.6077 -0.0534 -0.1373>, rot=<0.0082 0.0188 0.1708 -0.9851>}
 ring_metacarpal_r {parent=hand_r, loc=<-2.2074 0.3625 -1.3638>, rot=<-0.0284 0.0722 -0.0458 -0.9959>}
 ring_01_r {parent=ring_metacarpal_r, loc=<-7.0295 0.6197 0.1446>, rot=<-0.0489 -0.0252 0.1892 -0.9804>}
 ring_02_r {parent=ring_01_r, loc=<-3.6174 -0.4647 0.1027>, rot=<0 0 0.2806 -0.9598>}
 ring_03_r {parent=ring_02_r, loc=<-2.4152 -0.114 -0.0199>, rot=<0 0 0.1189 -0.9929>}
 pinky_metacarpal_r {parent=hand_r, loc=<-1.7616 0.9753 -3.3814>, rot=<-0.0397 0.0814 -0.069 -0.9935>}
 pinky_01_r {parent=pinky_metacarpal_r, loc=<-7.1709 0.4831 0.1094>, rot=<-0.0654 -0.0035 0.2242 -0.9723>}
 pinky_02_r {parent=pinky_01_r, loc=<-2.8274 -0.0893 0.1043>, rot=<0.0037 0.0011 0.2854 -0.9584>}
 pinky_03_r {parent=pinky_02_r, loc=<-2.2046 -0.0307 0.0078>, rot=<-0.0002 0.0098 0.1445 -0.9895>}
 deform_wrist_dn_base_r {parent=hand_r, loc=<0 3.4172 0>, rot=<0 0 0 -1>}
 deform_wrist_dn_r {parent=deform_wrist_dn_base_r, loc=<0 0 0>, rot=<0 0 0 -1>}
 deform_wrist_up_base_r {parent=hand_r, loc=<-0.0001 -2.7804 0>, rot=<0 0 0 -1>}
 deform_wrist_up_r {parent=deform_wrist_up_base_r, loc=<0 0 0>, rot=<0 0 0 -1>}
 lowerarm_twist_02_r {parent=lowerarm_r, loc=<-9.6199 0 0>, rot=<0 0 0 -1>}
 lowerarm_twist_01_r {parent=lowerarm_r, loc=<-17.8549 0 0>, rot=<0 0 0 -1>}
 deform_elbow_in_base_r {parent=lowerarm_r, loc=<-0.2114 -5.6449 -0.827>, rot=<0 0 0 -1>}
 deform_elbow_in_r {parent=deform_elbow_in_base_r, loc=<0 0 0>, rot=<0 0 0 1>}
 deform_elbow_out_base_r {parent=lowerarm_r, loc=<-0.0001 7.5129 0.0002>, rot=<0 0 0 -1>}
 deform_elbow_out_r {parent=deform_elbow_out_base_r, loc=<0 0 0>, rot=<0 0 0 1>}
 dyn_fluffB_r_MSTR {parent=lowerarm_r, loc=<-4.3418 5.1628 -3.7825>, rot=<0.8857 0.1113 -0.1666 -0.4189>}
 dyn_fluffB_r_01 {parent=dyn_fluffB_r_MSTR, loc=<0 0 0>, rot=<0 0 0 1>}
 dyn_fluffA_r_MSTR {parent=lowerarm_r, loc=<-3.1503 2.5871 -6.7776>, rot=<0.9514 0.0771 -0.1615 -0.2508>}
 dyn_fluffA_r_01 {parent=dyn_fluffA_r_MSTR, loc=<0 0 0>, rot=<0 0 0 1>}
 upperarm_twist_01_r {parent=upperarm_r, loc=<-7.9796 -0.0843 0.6195>, rot=<0 0 0 -1>}
 deform_upperarm_bk_base_r {parent=upperarm_twist_01_r, loc=<6.8985 13.3826 1.3257>, rot=<0 0 0 1>}
 deform_upperarm_bk_r {parent=deform_upperarm_bk_base_r, loc=<0 0 0>, rot=<0 0 0 1>}
 deform_upperarm_fr_base_r {parent=upperarm_twist_01_r, loc=<8.3485 -10.7117 2.1456>, rot=<0 0 0 1>}
 deform_upperarm_fr_r {parent=deform_upperarm_fr_base_r, loc=<0 0 0>, rot=<0 0 0 1>}
 deform_upperarm_up_base_r {parent=upperarm_twist_01_r, loc=<11.2015 1.2234 -8.7788>, rot=<0 0 0 1>}
 deform_upperarm_up_r {parent=deform_upperarm_up_base_r, loc=<0 0 0>, rot=<0 0 0 1>}
 upperarm_twist_02_r {parent=upperarm_r, loc=<-19.5059 -0.0843 0.6195>, rot=<0 0 0 -1>}
 deform_bicep_base_r {parent=upperarm_twist_02_r, loc=<0.0002 -8.0494 -0.0003>, rot=<0 0 0 1>}
 deform_bicep_r {parent=deform_bicep_base_r, loc=<0 0 0>, rot=<0 0 0 1>}
 dyn_scarfFlap_l_MSTR {parent=spine_05, loc=<12.7936 -11.2179 -8.6628>, rot=<0.3762 -0.9249 -0.0522 -0.0165>}
 dyn_scarfFlap_l_01 {parent=dyn_scarfFlap_l_MSTR, loc=<0 0 0>, rot=<0 0 0 1>}
 dyn_scarfFlap_l_02 {parent=dyn_scarfFlap_l_01, loc=<5.7131 0 0>, rot=<-0.023 -0.0118 -0.2157 -0.9761>}
 dyn_scarfFlap_l_03 {parent=dyn_scarfFlap_l_02, loc=<5.6556 0 0>, rot=<0 0 0 1>}
 dyn_scarf_c_MSTR {parent=spine_05, loc=<8.9924 12.218 -0.0816>, rot=<0.017 -0.0636 0.9815 0.1798>}
 dyn_scarf_c_01 {parent=dyn_scarf_c_MSTR, loc=<0 0 0>, rot=<0 0 0 1>}
 dyn_scarf_c_02 {parent=dyn_scarf_c_01, loc=<10.2425 0 0>, rot=<0 0 0 1>}
 dyn_scarNeck_r_MSTR {parent=spine_05, loc=<22.5309 -6.3973 9.8268>, rot=<0.2857 -0.6564 0.105 -0.6903>}
 dyn_scarNeck_r_01 {parent=dyn_scarNeck_r_MSTR, loc=<0 0 0>, rot=<0 0 0 1>}
 dyn_scarNeck_l_MSTR {parent=spine_05, loc=<22.8724 -6.1423 -8.5267>, rot=<-0.1381 -0.699 -0.3696 -0.5964>}
 dyn_scarNeck_l_01 {parent=dyn_scarNeck_l_MSTR, loc=<0 0 0>, rot=<0 0 0 1>}
 dyn_scarNeck_c_MSTR {parent=spine_05, loc=<23.5011 -8.0763 0.053>, rot=<0.0284 -0.7065 -0.0284 -0.7065>}
 dyn_scarNeck_c_01 {parent=dyn_scarNeck_c_MSTR, loc=<0 0 0>, rot=<0 0 0 1>}
 deform_pec_aim_base_l {parent=spine_04, loc=<5.0227 13.6719 -2.6508>, rot=<0.0672 -0.5201 0.3228 -0.7879>}
 deform_pec_aim_l {parent=deform_pec_aim_base_l, loc=<0 0 0>, rot=<0 0 0 1>}
 deform_pec_base_l {parent=deform_pec_aim_l, loc=<16.3279 0 0>, rot=<0 0 0 1>}
 deform_pec_l {parent=deform_pec_base_l, loc=<0 0 0>, rot=<0 0 0 1>}
 deform_pec_aim_base_r {parent=spine_04, loc=<5.0225 13.6718 2.6508>, rot=<-0.5201 -0.0672 0.7879 0.3228>}
 deform_pec_aim_r {parent=deform_pec_aim_base_r, loc=<0 0 0>, rot=<0 0 0 1>}
 deform_pec_base_r {parent=deform_pec_aim_r, loc=<-16.3279 0 0>, rot=<0 0 0 -1>}
 deform_pec_r {parent=deform_pec_base_r, loc=<0 0 0.0002>, rot=<0 0 0 -1>}
 deform_lat_l_base {parent=spine_03, loc=<2.5561 -10.3716 -17.6918>, rot=<0 0 0 -1>}
 deform_lat_l {parent=deform_lat_l_base, loc=<0 0 0>, rot=<0 0 0 1>}
 deform_lat_r_base {parent=spine_03, loc=<2.5562 -10.3716 17.6918>, rot=<0 0 1 0>}
 deform_lat_r {parent=deform_lat_r_base, loc=<0 0 0>, rot=<0 0 0 -1>}
 dyn_shirtJiggle_c_01 {parent=spine_02, loc=<2.8409 -15.3389 0>, rot=<0.0399 -0.9992 0 0>}
 thigh_l {parent=pelvis, loc=<-1.8716 -0.0109 -11.9936>, rot=<0.0012 0.0572 0.0155 -0.9982>}
 calf_l {parent=thigh_l, loc=<-41.9178 -0.2246 0.0926>, rot=<0.0006 -0.0161 -0.0391 -0.9991>}
 foot_l {parent=calf_l, loc=<-38.7057 1.2438 0.2452>, rot=<0.0975 -0.0474 0.0317 -0.9936>}
 ball_l {parent=foot_l, loc=<-6.1614 13.1469 0.6807>, rot=<0.0001 -0.0001 -0.7063 -0.7079>}
 calf_twist_02_l {parent=calf_l, loc=<-15.9268 -0.0967 0.0397>, rot=<0.0237 0.0095 -0.0035 -0.9997>}
 dyn_legFluffA_l_MSTR {parent=calf_twist_02_l, loc=<-4.0954 -7.8221 -3.0726>, rot=<-0.1855 -0.2211 0.5496 -0.784>}
 dyn_legFluffA_l_01 {parent=dyn_legFluffA_l_MSTR, loc=<0 0 0>, rot=<0 0 0 1>}
 dyn_legFluffB_l_MSTR {parent=calf_twist_02_l, loc=<-4.614 -10.2606 0.0677>, rot=<0.0425 0.0295 0.4438 -0.8946>}
 dyn_legFluffB_l_01 {parent=dyn_legFluffB_l_MSTR, loc=<0 0 0>, rot=<0 0 0 1>}
 dyn_legFluffC_l_MSTR {parent=calf_twist_02_l, loc=<-4.0863 -8.3177 3.8989>, rot=<0.1961 0.1608 0.5284 -0.8102>}
 dyn_legFluffC_l_01 {parent=dyn_legFluffC_l_MSTR, loc=<0 0 0>, rot=<0 0 0 1>}
 calf_twist_01_l {parent=calf_l, loc=<-32.9335 -0.0967 0.0397>, rot=<0.0237 0.0095 -0.0035 -0.9997>}
 deform_knee_in_base_l {parent=calf_l, loc=<1.298 -8.0356 -0.0496>, rot=<0 0 0 -1>}
 deform_knee_in_l {parent=deform_knee_in_base_l, loc=<0 0 0>, rot=<0 0 0 1>}
 deform_knee_out_base_l {parent=calf_l, loc=<0 6.8739 0>, rot=<0 0 0 -1>}
 deform_knee_out_l {parent=deform_knee_out_base_l, loc=<0 0 0>, rot=<0 0 0 1>}
 thigh_twist_01_l {parent=thigh_l, loc=<-21.6303 0 0>, rot=<0.0013 0 0 -1>}
 deform_groin_base_l {parent=thigh_l, loc=<-2.1908 8.5661 -1.7442>, rot=<0 0 0 -1>}
 deform_groin_l {parent=deform_groin_base_l, loc=<0 0 0>, rot=<0 0 0 1>}
 thigh_r {parent=pelvis, loc=<-1.8716 -0.0109 11.9936>, rot=<0.0572 -0.0012 0.9982 0.0155>}
 calf_r {parent=thigh_r, loc=<41.9178 0.2246 -0.0926>, rot=<0.0006 -0.0161 -0.0391 -0.9991>}
 foot_r {parent=calf_r, loc=<38.7057 -1.2438 -0.2452>, rot=<0.0975 -0.0474 0.0317 -0.9936>}
 ball_r {parent=foot_r, loc=<6.1614 -13.1469 -0.6807>, rot=<0.0001 -0.0001 -0.7063 -0.7079>}
 calf_twist_02_r {parent=calf_r, loc=<15.9268 0.0967 -0.0397>, rot=<0.0237 0.0095 -0.0035 -0.9997>}
 dyn_legFluffA_r_MSTR {parent=calf_twist_02_r, loc=<3.2896 7.8883 3.529>, rot=<-0.2069 0.3129 -0.7765 -0.5063>}
 dyn_legFluffA_r_01 {parent=dyn_legFluffA_r_MSTR, loc=<0 0 0>, rot=<0 0 0 1>}
 dyn_legFluffC_r_MSTR {parent=calf_twist_02_r, loc=<3.4385 9.2634 -3.9254>, rot=<-0.0403 0.1618 0.8807 0.4433>}
 dyn_legFluffC_r_01 {parent=dyn_legFluffC_r_MSTR, loc=<0 0 0>, rot=<0 0 0 1>}
 dyn_legFluffB_r_MSTR {parent=calf_twist_02_r, loc=<2.9746 10.2875 0.9417>, rot=<0.0614 -0.0721 0.9098 0.404>}
 dyn_legFluffB_r_01 {parent=dyn_legFluffB_r_MSTR, loc=<0 0 0>, rot=<0 0 0 1>}
 calf_twist_01_r {parent=calf_r, loc=<32.9335 0.0967 -0.0397>, rot=<0.0237 0.0095 -0.0035 -0.9997>}
 deform_knee_in_base_r {parent=calf_r, loc=<-1.2981 8.0356 0.0496>, rot=<0 0 0 -1>}
 deform_knee_in_r {parent=deform_knee_in_base_r, loc=<0 0 0>, rot=<0 0 0 1>}
 deform_knee_out_base_r {parent=calf_r, loc=<0 -6.8739 0>, rot=<0 0 0 -1>}
 deform_knee_out_r {parent=deform_knee_out_base_r, loc=<0 0 0>, rot=<0 0 0 1>}
 thigh_twist_01_r {parent=thigh_r, loc=<21.6303 0 0>, rot=<0.0013 0 0 -1>}
 deform_groin_base_r {parent=thigh_r, loc=<2.1908 -8.5661 1.7442>, rot=<0 0 0 -1>}
 deform_groin_r {parent=deform_groin_base_r, loc=<0 0 0>, rot=<0 0 0 -1>}
 deform_glute_base_l {parent=pelvis, loc=<-10.1877 -11.6217 -12.899>, rot=<0.0275 0.0501 -0.4472 -0.8926>}
 deform_glute_l {parent=deform_glute_base_l, loc=<0 0 0>, rot=<0 0 0 1>}
 deform_glute_base_r {parent=pelvis, loc=<-10.1877 -11.6218 12.899>, rot=<0.0501 -0.0275 0.8926 -0.4472>}
 deform_glute_r {parent=deform_glute_base_r, loc=<0 0 0>, rot=<0 0 0 1>}
 dyn_skirtA_l_MSTR {parent=pelvis, loc=<1.3325 10.8167 -7.7678>, rot=<-0.0192 0.1818 0.981 0.0648>}
 dyn_skirtA_l_01 {parent=dyn_skirtA_l_MSTR, loc=<0 0 0>, rot=<0 0 0 1>}
 dyn_skirtA_l_02 {parent=dyn_skirtA_l_01, loc=<10.1105 0 0>, rot=<0.0209 -0.0298 0.0302 -0.9989>}
 dyn_skirtA_l_03 {parent=dyn_skirtA_l_02, loc=<10.7281 0 0>, rot=<0.0237 -0.0033 0.0153 -0.9996>}
 dyn_skirtA_l_04 {parent=dyn_skirtA_l_03, loc=<13.6 0 0>, rot=<0 0 0 1>}
 dyn_skirtB_l_MSTR {parent=pelvis, loc=<3.4195 4.7003 -15.9541>, rot=<-0.1036 0.4871 0.8617 0.0971>}
 dyn_skirtB_l_01 {parent=dyn_skirtB_l_MSTR, loc=<0 0 0>, rot=<0 0 0 1>}
 dyn_skirtB_l_02 {parent=dyn_skirtB_l_01, loc=<10.1105 0 0>, rot=<0 0 0 1>}
 dyn_skirtB_l_03 {parent=dyn_skirtB_l_02, loc=<10.7281 0 0>, rot=<0.0238 0 0 -0.9997>}
 dyn_skirtB_l_04 {parent=dyn_skirtB_l_03, loc=<13.6 0 0>, rot=<0 0 0 1>}
 dyn_skirtD_l_MSTR {parent=pelvis, loc=<3.2661 -12.7111 -10.0678>, rot=<0.1149 -0.9931 -0.0164 -0.0191>}
 dyn_skirtD_l_01 {parent=dyn_skirtD_l_MSTR, loc=<0 0 0>, rot=<0 0 0 1>}
 dyn_skirtD_l_02 {parent=dyn_skirtD_l_01, loc=<9.7277 0 0>, rot=<0 0 0 1>}
 dyn_skirtD_l_03 {parent=dyn_skirtD_l_02, loc=<9.7277 0 0>, rot=<0.0238 0 0 -0.9997>}
 dyn_skirtD_l_04 {parent=dyn_skirtD_l_03, loc=<9.7277 0 0>, rot=<0 0 0 1>}
 dyn_skirtC_l_MSTR {parent=pelvis, loc=<3.2748 -5.1307 -17.8339>, rot=<0.1324 -0.8236 -0.5472 -0.0692>}
 dyn_skirtC_l_01 {parent=dyn_skirtC_l_MSTR, loc=<0 0 0>, rot=<0 0 0 1>}
 dyn_skirtC_l_02 {parent=dyn_skirtC_l_01, loc=<9.0416 0 0>, rot=<0 0 0 1>}
 dyn_skirtC_l_03 {parent=dyn_skirtC_l_02, loc=<9.6592 0 0>, rot=<0.0238 0 0 -0.9997>}
 dyn_skirtC_l_04 {parent=dyn_skirtC_l_03, loc=<12.5311 0 0>, rot=<0 0 0 1>}
 dyn_skirtA_r_MSTR {parent=pelvis, loc=<1.3313 10.8082 7.7645>, rot=<0.0158 -0.1821 0.9821 0.0462>}
 dyn_skirtA_r_01 {parent=dyn_skirtA_r_MSTR, loc=<0 0 0>, rot=<0 0 0 1>}
 dyn_skirtA_r_02 {parent=dyn_skirtA_r_01, loc=<10.1105 0 0>, rot=<-0.0209 0.0298 0.0302 -0.9989>}
 dyn_skirtA_r_03 {parent=dyn_skirtA_r_02, loc=<10.728 0 0>, rot=<0 0 0 1>}
 dyn_skirtA_r_04 {parent=dyn_skirtA_r_03, loc=<13.6 0 0>, rot=<0 0 0 -1>}
 dyn_skirtB_r_MSTR {parent=pelvis, loc=<3.4196 4.7003 15.9541>, rot=<0.0967 -0.4885 0.863 0.0848>}
 dyn_skirtB_r_01 {parent=dyn_skirtB_r_MSTR, loc=<0 0 0>, rot=<0 0 0 1>}
 dyn_skirtB_r_02 {parent=dyn_skirtB_r_01, loc=<10.1105 0 0>, rot=<0 0 0 1>}
 dyn_skirtB_r_03 {parent=dyn_skirtB_r_02, loc=<10.728 0 0>, rot=<-0.0238 0 0 -0.9997>}
 dyn_skirtB_r_04 {parent=dyn_skirtB_r_03, loc=<13.6 0.0001 -0.0001>, rot=<0 0 0 1>}
 dyn_skirtC_r_MSTR {parent=pelvis, loc=<3.2748 -5.1307 17.8339>, rot=<0.1324 -0.8236 0.5472 0.0692>}
 dyn_skirtC_r_01 {parent=dyn_skirtC_r_MSTR, loc=<0 0 0>, rot=<0 0 0 1>}
 dyn_skirtC_r_02 {parent=dyn_skirtC_r_01, loc=<9.0416 0 0>, rot=<0 0 0 1>}
 dyn_skirtC_r_03 {parent=dyn_skirtC_r_02, loc=<9.6592 0 0>, rot=<-0.0238 0 0 -0.9997>}
 dyn_skirtC_r_04 {parent=dyn_skirtC_r_03, loc=<12.5311 0 0>, rot=<0 0 0 1>}
 dyn_skirtD_r_MSTR {parent=pelvis, loc=<3.2661 -12.711 10.0678>, rot=<0.1149 -0.9931 0.0164 0.0191>}
 dyn_skirtD_r_01 {parent=dyn_skirtD_r_MSTR, loc=<0 0 0>, rot=<0 0 0 1>}
 dyn_skirtD_r_02 {parent=dyn_skirtD_r_01, loc=<9.7276 -0.0001 0.0001>, rot=<0 0 0 1>}
 dyn_skirtD_r_03 {parent=dyn_skirtD_r_02, loc=<9.7277 0 0>, rot=<-0.0238 0 0 -0.9997>}
 dyn_skirtD_r_04 {parent=dyn_skirtD_r_03, loc=<9.7276 0 0>, rot=<0 0 0 1>}
 dyn_bowMain_l_MSTR {parent=pelvis, loc=<6.8613 13.0481 -4.2962>, rot=<-0.0178 -0.9929 0.1172 0.0021>}
 dyn_bowMain_l_01 {parent=dyn_bowMain_l_MSTR, loc=<0 0 0>, rot=<0 0 0 1>}
 dyn_bowD_l_MSTR {parent=dyn_bowMain_l_01, loc=<-0.2495 0.1321 -1.8603>, rot=<0.9617 0.1177 -0.2314 0.0885>}
 dyn_bowD_l_01 {parent=dyn_bowD_l_MSTR, loc=<0 0 0>, rot=<0 0 0 1>}
 dyn_bowD_l_02 {parent=dyn_bowD_l_01, loc=<6.4104 0 0>, rot=<0.0948 -0.0745 -0.105 -0.9871>}
 dyn_bowC_l_MSTR {parent=dyn_bowMain_l_01, loc=<-0.5676 0.315 1.356>, rot=<0.8967 0.2927 0.1987 -0.2661>}
 dyn_bowC_l_01 {parent=dyn_bowC_l_MSTR, loc=<0 0 0>, rot=<0 0 0 1>}
 dyn_bowC_l_02 {parent=dyn_bowC_l_01, loc=<5.4718 0 0>, rot=<-0.0031 0.047 -0.0658 -0.9967>}
 dyn_bowB_l_MSTR {parent=dyn_bowMain_l_01, loc=<1.5425 -0.4641 -0.7442>, rot=<0.9965 0.0587 0.0177 0.0565>}
 dyn_bowB_l_01 {parent=dyn_bowB_l_MSTR, loc=<0 0 0>, rot=<0 0 0 1>}
 dyn_bowB_l_02 {parent=dyn_bowB_l_01, loc=<9.1465 0 0>, rot=<0.0403 -0.0238 0.0329 -0.9984>}
 dyn_bowB_l_03 {parent=dyn_bowB_l_02, loc=<9.1755 0 0>, rot=<-0.0427 -0.0213 0.0171 -0.9987>}
 dyn_bowB_l_04 {parent=dyn_bowB_l_03, loc=<9.8241 0 0>, rot=<0 0 0.0196 -0.9998>}
 dyn_bowA_l_MSTR {parent=dyn_bowMain_l_01, loc=<1.4907 -0.0022 0.6332>, rot=<0.9728 0.1657 0.1387 -0.084>}
 dyn_bowA_l_01 {parent=dyn_bowA_l_MSTR, loc=<0 0 0>, rot=<0 0 0 1>}
 dyn_bowA_l_02 {parent=dyn_bowA_l_01, loc=<9.1071 0 0>, rot=<0.0817 -0.0402 0.0072 -0.9958>}
 dyn_bowA_l_03 {parent=dyn_bowA_l_02, loc=<8.4582 0 0>, rot=<0 0 0 1>}
 dyn_bentoBagMain_Pivot_MSTR {parent=pelvis, loc=<3.5594 -5.1384 10.1433>, rot=<0.5509 -0.7663 0.2614 0.2025>}
 dyn_bentoBagMain_Pivot {parent=dyn_bentoBagMain_Pivot_MSTR, loc=<0 0 0>, rot=<0 0 0 1>}
 dyn_bentoBagMain_MSTR {parent=dyn_bentoBagMain_Pivot, loc=<11.3869 0 0>, rot=<0.5967 0.1752 -0.3915 -0.6782>}
 dyn_bentoBagMain_01 {parent=dyn_bentoBagMain_MSTR, loc=<0 0 0>, rot=<0 0 0 1>}
 dyn_bentoBagF_MSTR {parent=dyn_bentoBagMain_01, loc=<0.9229 0.0386 -1.6945>, rot=<-0.2165 -0.1978 -0.191 -0.9368>}
 dyn_bentoBagF_01 {parent=dyn_bentoBagF_MSTR, loc=<0 0 0>, rot=<0 0 0 1>}
 dyn_bentoBagE_MSTR {parent=dyn_bentoBagMain_01, loc=<0.5654 -1.1023 -1.2235>, rot=<0.2076 -0.2978 0.4173 -0.8331>}
 dyn_bentoBagE_01 {parent=dyn_bentoBagE_MSTR, loc=<0 0 0>, rot=<0 0 0 1>}
 dyn_bentoBagD_MSTR {parent=dyn_bentoBagMain_01, loc=<0.409 0.2996 -1.3301>, rot=<0.0835 -0.1517 -0.6632 -0.7282>}
 dyn_bentoBagD_01 {parent=dyn_bentoBagD_MSTR, loc=<0 0 0>, rot=<0 0 0 1>}
 dyn_bentoBagC_MSTR {parent=dyn_bentoBagMain_01, loc=<-0.0129 -0.8289 -1.4828>, rot=<-0.135 -0.287 0.8484 -0.4238>}
 dyn_bentoBagC_01 {parent=dyn_bentoBagC_MSTR, loc=<0 0 0>, rot=<0 0 0 1>}
 dyn_bentoBagB_MSTR {parent=dyn_bentoBagMain_01, loc=<-1.2271 -0.237 -0.1305>, rot=<-0.0157 -0.7956 0.488 -0.3588>}
 dyn_bentoBagB_01 {parent=dyn_bentoBagB_MSTR, loc=<0 0 0>, rot=<0 0 0 1>}
 dyn_bentoBagA_MSTR {parent=dyn_bentoBagMain_01, loc=<-1.1584 1.3379 -0.1879>, rot=<-0.3704 -0.5617 -0.5487 -0.4962>}
 dyn_bentoBagA_01 {parent=dyn_bentoBagA_MSTR, loc=<0 0 0>, rot=<0 0 0 1>}
 dyn_bentoRopeAim_MSTR {parent=pelvis, loc=<3.2685 -13.8054 2.2224>, rot=<0.0177 0.7795 0.0083 -0.6261>}
 dyn_bentoRopeAim_01 {parent=dyn_bentoRopeAim_MSTR, loc=<0 0 0>, rot=<0 0 0 1>}
]
variables [
 is_not_crouching {type=bool, default=False}
 is_not_skydiving {type=bool, default=False}
 is_player_moving {type=double, default=0}
 thigh_l_up {type=float, default=0}
 thigh_r_out {type=float, default=0}
 thigh_r_up {type=float, default=0}
]
