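# Physics asset for the large male HealedScar player body, built for a dynamic skirt over lower-body collision.
# Capsules cover spine_02, pelvis, thighs, calves and feet, linked by uniform angular limits.
# Eight skirt chains (A–D, left/right) hang from pelvis via tiny angularly free MSTR bodies, each with four capsule segments using per-chain swing limits and zero twist.
# Right A chain has two extra angularly free constraints tying segments 02 and 03 to 04.
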PhysicsAsset /BRCosmetics/Characters/Player/Male/Large/Bodies/M_LRG_HealedScar/Meshes/M_LRG_HealedScar_Physics
bodies {
 calf_l [{shapeType=Capsule, length=21.5021, radius=13.1187, center=<-17.4038 -2.9931 1.6127>, rotation=<rot 90 180 180>}]
 calf_r [{shapeType=Capsule, length=33.0036, radius=14.7465, center=<17.4023 2.2595 -1.8925>, rotation=<rot 90 180 180>}]
 dyn_skirtA_l_01 [{shapeType=Capsule, length=17, radius=6.5, shapeName=dyn_skirtA_l_01_capsule, center=<4 -6 -8>, rotation=<rot 0 -0.0001 -15.0388>}]
 dyn_skirtA_l_02 [{shapeType=Capsule, length=16, radius=7.5, shapeName=dyn_skirtA_l_02_capsule, center=<4 -5.1749 -10>, rotation=<rot 0 -0.0001 -16.9193>}]
 dyn_skirtA_l_03 [{shapeType=Capsule, length=14, radius=9.5, shapeName=dyn_skirtA_l_03_capsule, center=<4 -5.1749 -10>, rotation=<rot 0 -0.0001 -16.9193>}]
 dyn_skirtA_l_04 [{shapeType=Capsule, length=13.5, radius=10, shapeName=dyn_skirtA_l_04_capsule, center=<4 -5.1749 -10>, rotation=<rot 0 -0.0001 -16.9193>}]
 dyn_skirtA_l_MSTR [{shapeType=Capsule, length=0.505, radius=0.505, shapeName=dyn_skirtA_l_MSTR_capsule}]
 dyn_skirtA_r_01 [{shapeType=Capsule, length=10.5, radius=6.5, shapeName=dyn_skirtA_r_01_capsule, center=<4 -8 0>}]
 dyn_skirtA_r_02 [{shapeType=Capsule, length=9.5, radius=7.5, shapeName=dyn_skirtA_r_02_capsule, center=<4 -8 0>}]
 dyn_skirtA_r_03 [{shapeType=Capsule, length=8, radius=9, shapeName=dyn_skirtA_r_03_capsule, center=<5.8893 -8 0>}]
 dyn_skirtA_r_04 [{shapeType=Capsule, length=7, radius=10, shapeName=dyn_skirtA_r_04_capsule, center=<5.6179 -8 0>}]
 dyn_skirtA_r_MSTR [{shapeType=Capsule, length=0.505, radius=0.505, shapeName=dyn_skirtA_r_MSTR_capsule}]
 dyn_skirtB_l_01 [{shapeType=Capsule, length=9, radius=8, shapeName=dyn_skirtB_l_01_capsule, center=<4 -8 0>}]
 dyn_skirtB_l_02 [{shapeType=Capsule, length=14, radius=8, shapeName=dyn_skirtB_l_02_capsule, center=<6.3776 -8 0>, rotation=<rot 0 0 9.8661>}]
 dyn_skirtB_l_03 [{shapeType=Capsule, length=17.5, radius=9.5, shapeName=dyn_skirtB_l_03_capsule, center=<6.3776 -8 0>, rotation=<rot 0.0001 -0.0001 5.1713>}]
 dyn_skirtB_l_04 [{shapeType=Capsule, length=16, radius=10.5, shapeName=dyn_skirtB_l_04_capsule, center=<6.3776 -8 0>, rotation=<rot 0.0001 -0.0001 5.1713>}]
 dyn_skirtB_l_MSTR [{shapeType=Capsule, length=0.505, radius=0.505, shapeName=dyn_skirtB_l_MSTR_capsule}]
 dyn_skirtB_r_01 [{shapeType=Capsule, length=11, radius=8, shapeName=dyn_skirtB_r_01_capsule, center=<4 -8 0>}]
 dyn_skirtB_r_02 [{shapeType=Capsule, length=9.5, radius=9.5, shapeName=dyn_skirtB_r_02_capsule, center=<4 -8 0>}]
 dyn_skirtB_r_03 [{shapeType=Capsule, length=8.5, radius=10.5, shapeName=dyn_skirtB_r_03_capsule, center=<4 -8 0>}]
 dyn_skirtB_r_04 [{shapeType=Capsule, length=7.5, radius=11.5, shapeName=dyn_skirtB_r_04_capsule, center=<4 -8 0>}]
 dyn_skirtB_r_MSTR [{shapeType=Capsule, length=0.505, radius=0.505, shapeName=dyn_skirtB_r_MSTR_capsule}]
 dyn_skirtC_l_01 [{shapeType=Capsule, length=9, radius=8, shapeName=dyn_skirtC_l_01_capsule, center=<4 -8 0>}]
 dyn_skirtC_l_02 [{shapeType=Capsule, length=9, radius=8, shapeName=dyn_skirtC_l_02_capsule, center=<4 -8 0>}]
 dyn_skirtC_l_03 [{shapeType=Capsule, length=9, radius=8, shapeName=dyn_skirtC_l_03_capsule, center=<4 -8 0>}]
 dyn_skirtC_l_04 [{shapeType=Capsule, length=9, radius=8, shapeName=dyn_skirtC_l_04_capsule, center=<4 -8 0>}]
 dyn_skirtC_l_MSTR [{shapeType=Capsule, length=0.505, radius=0.505, shapeName=dyn_skirtC_l_MSTR_capsule}]
 dyn_skirtC_r_01 [{shapeType=Capsule, length=9, radius=8, shapeName=dyn_skirtC_r_01_capsule, center=<4 -8 0>}]
 dyn_skirtC_r_02 [{shapeType=Capsule, length=9, radius=8, shapeName=dyn_skirtC_r_02_capsule, center=<4 -8 0>}]
 dyn_skirtC_r_03 [{shapeType=Capsule, length=18, radius=8, shapeName=dyn_skirtC_r_03_capsule, center=<4 -8 0>, rotation=<rot 0 0 11.6188>}]
 dyn_skirtC_r_04 [{shapeType=Capsule, length=18, radius=8, shapeName=dyn_skirtC_r_04_capsule, center=<4 -8 0>, rotation=<rot 0 0 11.6188>}]
 dyn_skirtC_r_MSTR [{shapeType=Capsule, length=0.505, radius=0.505, shapeName=dyn_skirtC_r_MSTR_capsule}]
 dyn_skirtD_l_01 [{shapeType=Capsule, length=9, radius=8, shapeName=dyn_skirtD_l_01_capsule, center=<4 -8 0>}]
 dyn_skirtD_l_02 [{shapeType=Capsule, length=9, radius=8, shapeName=dyn_skirtD_l_02_capsule, center=<4 -8 0>}]
 dyn_skirtD_l_03 [{shapeType=Capsule, length=18, radius=8, shapeName=dyn_skirtD_l_03_capsule, center=<4 -8 0>}]
 dyn_skirtD_l_04 [{shapeType=Capsule, length=17, radius=9, shapeName=dyn_skirtD_l_04_capsule, center=<4 -8 0>}]
 dyn_skirtD_l_MSTR [{shapeType=Capsule, length=0.505, radius=0.505, shapeName=dyn_skirtD_l_MSTR_capsule}]
 dyn_skirtD_r_01 [{shapeType=Capsule, length=9, radius=8, shapeName=dyn_skirtD_r_01_capsule, center=<4 -8 0>}]
 dyn_skirtD_r_02 [{shapeType=Capsule, length=9, radius=8, shapeName=dyn_skirtD_r_02_capsule, center=<4 -8 0>}]
 dyn_skirtD_r_03 [{shapeType=Capsule, length=18, radius=8, shapeName=dyn_skirtD_r_03_capsule, center=<4 -8 0>}]
 dyn_skirtD_r_04 [{shapeType=Capsule, length=22, radius=9, shapeName=dyn_skirtD_r_04_capsule, center=<4 -8 0>, rotation=<rot 0 0.0001 -13.9114>}]
 dyn_skirtD_r_MSTR [{shapeType=Capsule, length=0.505, radius=0.505, shapeName=dyn_skirtD_r_MSTR_capsule}]
 foot_l [{shapeType=Capsule, length=18.9985, radius=13.1556, center=<-5.8088 8.36 1.3184>, rotation=<rot 0 -0.0001 -94.1007>}]
 foot_r [{shapeType=Capsule, length=18.9985, radius=14.1556, center=<5.8088 -8.36 -1.3184>, rotation=<rot 0 -0.0001 85.8993>}]
 pelvis [{shapeType=Capsule, length=12.0574, radius=11.6316, center=<0 -0.7191 0>}]
 spine_02 [{shapeType=Capsule, length=20.4185, radius=10.5451}]
 thigh_l [{shapeType=Capsule, length=33.2245, radius=13.7313, center=<-20.3026 -1.8869 4.9672>, rotation=<rot 85.5099 179.9995 -178.846>}]
 thigh_r [{shapeType=Capsule, length=36.7245, radius=13.7313, center=<22.2552 0.2922 -4.8139>, rotation=<rot 85.5099 179.9766 4.7613>}]
}
constraints [
 {bone1Name=spine_02, bone2Name=pelvis, swing1LimitDegrees=45, swing2LimitDegrees=45, twistLimitDegrees=45}
 {bone1Name=thigh_l, bone2Name=pelvis, swing1LimitDegrees=45, swing2LimitDegrees=45, twistLimitDegrees=45}
 {bone1Name=calf_l, bone2Name=thigh_l, swing1LimitDegrees=45, swing2LimitDegrees=45, twistLimitDegrees=45}
 {bone1Name=foot_l, bone2Name=calf_l, swing1LimitDegrees=45, swing2LimitDegrees=45, twistLimitDegrees=45}
 {bone1Name=calf_r, bone2Name=thigh_r, swing1LimitDegrees=45, swing2LimitDegrees=45, twistLimitDegrees=45}
 {bone1Name=thigh_r, bone2Name=pelvis, swing1LimitDegrees=45, swing2LimitDegrees=45, twistLimitDegrees=45}
 {
  bone1Name=dyn_skirtA_l_MSTR
  bone2Name=pelvis
  swing1LimitDegrees=45
  swing1Motion=Free
  swing2LimitDegrees=45
  swing2Motion=Free
  twistLimitDegrees=45
  twistMotion=Free
 }
 {bone1Name=dyn_skirtA_l_01, bone2Name=dyn_skirtA_l_MSTR, swing1LimitDegrees=25, swing2LimitDegrees=10, twistLimitDegrees=0}
 {bone1Name=dyn_skirtA_l_02, bone2Name=dyn_skirtA_l_01, swing1LimitDegrees=25, swing2LimitDegrees=10, twistLimitDegrees=0}
 {bone1Name=dyn_skirtA_l_03, bone2Name=dyn_skirtA_l_02, swing1LimitDegrees=30, swing2LimitDegrees=10, twistLimitDegrees=0}
 {bone1Name=dyn_skirtA_l_04, bone2Name=dyn_skirtA_l_03, swing1LimitDegrees=30, swing2LimitDegrees=10, twistLimitDegrees=0}
 {
  bone1Name=dyn_skirtA_r_MSTR
  bone2Name=pelvis
  swing1LimitDegrees=45
  swing1Motion=Free
  swing2LimitDegrees=45
  swing2Motion=Free
  twistLimitDegrees=45
  twistMotion=Free
 }
 {bone1Name=dyn_skirtA_r_01, bone2Name=dyn_skirtA_r_MSTR, swing1LimitDegrees=30, swing2LimitDegrees=20, twistLimitDegrees=0}
 {bone1Name=dyn_skirtA_r_02, bone2Name=dyn_skirtA_r_01, swing1LimitDegrees=20, swing2LimitDegrees=10, twistLimitDegrees=0}
 {bone1Name=dyn_skirtA_r_03, bone2Name=dyn_skirtA_r_02, swing1LimitDegrees=30, swing2LimitDegrees=10, twistLimitDegrees=0}
 {bone1Name=dyn_skirtA_r_04, bone2Name=dyn_skirtA_r_03, swing1LimitDegrees=30, swing2LimitDegrees=10, twistLimitDegrees=0}
 {
  bone1Name=dyn_skirtB_l_MSTR
  bone2Name=pelvis
  swing1LimitDegrees=45
  swing1Motion=Free
  swing2LimitDegrees=45
  swing2Motion=Free
  twistLimitDegrees=45
  twistMotion=Free
 }
 {bone1Name=dyn_skirtB_l_01, bone2Name=dyn_skirtB_l_MSTR, swing1LimitDegrees=15, swing2LimitDegrees=20, twistLimitDegrees=0}
 {bone1Name=dyn_skirtB_l_02, bone2Name=dyn_skirtB_l_01, swing1LimitDegrees=20, swing2LimitDegrees=10, twistLimitDegrees=0}
 {bone1Name=dyn_skirtB_l_03, bone2Name=dyn_skirtB_l_02, swing1LimitDegrees=30, swing2LimitDegrees=10, twistLimitDegrees=0}
 {bone1Name=dyn_skirtB_l_04, bone2Name=dyn_skirtB_l_03, swing1LimitDegrees=30, swing2LimitDegrees=10, twistLimitDegrees=0}
 {
  bone1Name=dyn_skirtB_r_MSTR
  bone2Name=pelvis
  swing1LimitDegrees=45
  swing1Motion=Free
  swing2LimitDegrees=45
  swing2Motion=Free
  twistLimitDegrees=45
  twistMotion=Free
 }
 {bone1Name=dyn_skirtB_r_01, bone2Name=dyn_skirtB_r_MSTR, swing1LimitDegrees=20, swing2LimitDegrees=10, twistLimitDegrees=0}
 {bone1Name=dyn_skirtB_r_02, bone2Name=dyn_skirtB_r_01, swing1LimitDegrees=20, swing2LimitDegrees=10, twistLimitDegrees=0}
 {bone1Name=dyn_skirtB_r_03, bone2Name=dyn_skirtB_r_02, swing1LimitDegrees=30, swing2LimitDegrees=10, twistLimitDegrees=0}
 {bone1Name=dyn_skirtB_r_04, bone2Name=dyn_skirtB_r_03, swing1LimitDegrees=30, swing2LimitDegrees=10, twistLimitDegrees=0}
 {
  bone1Name=dyn_skirtC_l_MSTR
  bone2Name=pelvis
  swing1LimitDegrees=45
  swing1Motion=Free
  swing2LimitDegrees=45
  swing2Motion=Free
  twistLimitDegrees=45
  twistMotion=Free
 }
 {bone1Name=dyn_skirtC_l_01, bone2Name=dyn_skirtC_l_MSTR, swing1LimitDegrees=20, swing2LimitDegrees=8, twistLimitDegrees=0}
 {bone1Name=dyn_skirtC_l_02, bone2Name=dyn_skirtC_l_01, swing1LimitDegrees=20, swing2LimitDegrees=8, twistLimitDegrees=0}
 {bone1Name=dyn_skirtC_l_03, bone2Name=dyn_skirtC_l_02, swing1LimitDegrees=20, swing2LimitDegrees=8, twistLimitDegrees=0}
 {bone1Name=dyn_skirtC_l_04, bone2Name=dyn_skirtC_l_03, swing1LimitDegrees=20, swing2LimitDegrees=8, twistLimitDegrees=0}
 {
  bone1Name=dyn_skirtC_r_MSTR
  bone2Name=pelvis
  swing1LimitDegrees=45
  swing1Motion=Free
  swing2LimitDegrees=45
  swing2Motion=Free
  twistLimitDegrees=45
  twistMotion=Free
 }
 {bone1Name=dyn_skirtC_r_01, bone2Name=dyn_skirtC_r_MSTR, swing1LimitDegrees=20, swing2LimitDegrees=15, twistLimitDegrees=0}
 {bone1Name=dyn_skirtC_r_02, bone2Name=dyn_skirtC_r_01, swing1LimitDegrees=20, swing2LimitDegrees=15, twistLimitDegrees=0}
 {bone1Name=dyn_skirtC_r_03, bone2Name=dyn_skirtC_r_02, swing1LimitDegrees=20, swing2LimitDegrees=15, twistLimitDegrees=0}
 {bone1Name=dyn_skirtC_r_04, bone2Name=dyn_skirtC_r_03, swing1LimitDegrees=20, swing2LimitDegrees=15, twistLimitDegrees=0}
 {
  bone1Name=dyn_skirtD_l_MSTR
  bone2Name=pelvis
  swing1LimitDegrees=45
  swing1Motion=Free
  swing2LimitDegrees=45
  swing2Motion=Free
  twistLimitDegrees=45
  twistMotion=Free
 }
 {bone1Name=dyn_skirtD_l_01, bone2Name=dyn_skirtD_l_MSTR, swing1LimitDegrees=12, swing2LimitDegrees=8, twistLimitDegrees=0}
 {bone1Name=dyn_skirtD_l_02, bone2Name=dyn_skirtD_l_01, swing1LimitDegrees=12, swing2LimitDegrees=8, twistLimitDegrees=0}
 {bone1Name=dyn_skirtD_l_03, bone2Name=dyn_skirtD_l_02, swing1LimitDegrees=12, swing2LimitDegrees=8, twistLimitDegrees=0}
 {bone1Name=dyn_skirtD_l_04, bone2Name=dyn_skirtD_l_03, swing1LimitDegrees=12, swing2LimitDegrees=8, twistLimitDegrees=0}
 {
  bone1Name=dyn_skirtD_r_MSTR
  bone2Name=pelvis
  swing1LimitDegrees=45
  swing1Motion=Free
  swing2LimitDegrees=45
  swing2Motion=Free
  twistLimitDegrees=45
  twistMotion=Free
 }
 {bone1Name=dyn_skirtD_r_01, bone2Name=dyn_skirtD_r_MSTR, swing1LimitDegrees=12, swing2LimitDegrees=8, twistLimitDegrees=0}
 {bone1Name=dyn_skirtD_r_02, bone2Name=dyn_skirtD_r_01, swing1LimitDegrees=12, swing2LimitDegrees=8, twistLimitDegrees=0}
 {bone1Name=dyn_skirtD_r_03, bone2Name=dyn_skirtD_r_02, swing1LimitDegrees=12, swing2LimitDegrees=8, twistLimitDegrees=0}
 {bone1Name=dyn_skirtD_r_04, bone2Name=dyn_skirtD_r_03, swing1LimitDegrees=12, swing2LimitDegrees=8, twistLimitDegrees=0}
 {bone1Name=foot_r, bone2Name=calf_r, swing1LimitDegrees=45, swing2LimitDegrees=45, twistLimitDegrees=45}
 {
  bone1Name=dyn_skirtA_r_02
  bone2Name=dyn_skirtA_r_04
  swing1LimitDegrees=45
  swing1Motion=Free
  swing2LimitDegrees=45
  swing2Motion=Free
  twistLimitDegrees=45
  twistMotion=Free
 }
 {
  bone1Name=dyn_skirtA_r_03
  bone2Name=dyn_skirtA_r_04
  swing1LimitDegrees=45
  swing1Motion=Free
  swing2LimitDegrees=45
  swing2Motion=Free
  twistLimitDegrees=45
  twistMotion=Free
 }
]
