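# Animation Blueprint for the HealedScar large male player body, adding secondary motion to its clothing and accessories.
# The copied parent pose runs through Control Rig, AnimDynamics for bag, scarf and fluff, RigidBody, skirt AnimDynamics, then a crouch-switched bow RigidBody pass.
# Additive bone offsets finish the pose for skydiving, hang gliding and front-end idle.
# The event graph updates gravity override, movement and state flags, disabling scarf-flap simulation for certain backblings.
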AnimBlueprint /BRCosmetics/Characters/Player/Male/Large/Bodies/M_LRG_HealedScar/Meshes/M_LRG_HealedScar_AnimBP
target_skeleton=&/M_LRG_HealedScar_Skeleton
graphs [
 AnimGraph {nodes [
   n1 = Get Custom Idle is Playing Float() -> (Custom Idle is Playing Float:float)
   n2 = Get has_hand_glider_and_is_skydiving() -> (has_hand_glider_and_is_skydiving:bool)
   n3 = Get bIsSkydiveFloating()
   n4 = "Transform (Modify) Bone - Bone: dyn_shoulderPad_l_01"!(
    ComponentPose=Local To Component(LocalPose=Use cached pose 'postSimCache'())
    bAlphaBoolEnabled:bool=@n3
    Translation="8,0,-2"
    Rotation="5,0,20"
    Scale="1,1,1"
   ) -> (Pose) {
    alphaInputType=Bool
    rotationMode=BMM_Additive
    rotationSpace=BCS_BoneSpace
    translationMode=BMM_Additive
    translationSpace=BCS_BoneSpace
    alphaBoolBlend {blendOption=Linear}
    boneToModify {boneName=dyn_shoulderPad_l_01}
    rotation {pitch=5, roll=20}
    translation {x=8, z=-2}
   }
   n5 = "Transform (Modify) Bone - Bone: dyn_shoulderPlateC_l_01"!(ComponentPose=@n4, bAlphaBoolEnabled:bool=@n3, Translation="-2,0,0", Rotation="0,0,0", Scale="1,1,1") -> (Pose) {
    alphaInputType=Bool
    rotationMode=BMM_Additive
    rotationSpace=BCS_BoneSpace
    translationMode=BMM_Additive
    translationSpace=BCS_BoneSpace
    alphaBoolBlend {blendOption=Linear}
    boneToModify {boneName=dyn_shoulderPlateC_l_01}
    translation {x=-2}
   }
   n6 = "Transform (Modify) Bone - Bone: dyn_shoulderPlateB_l_01"!(ComponentPose=@n5, bAlphaBoolEnabled:bool=@n3, Translation="-2,0,0", Rotation="0,0,0", Scale="1,1,1") -> (Pose) {
    alphaInputType=Bool
    rotationMode=BMM_Additive
    rotationSpace=BCS_BoneSpace
    translationMode=BMM_Additive
    translationSpace=BCS_BoneSpace
    alphaBoolBlend {blendOption=Linear}
    boneToModify {boneName=dyn_shoulderPlateB_l_01}
    translation {x=-2}
   }
   n7 = "Transform (Modify) Bone - Bone: dyn_scarfFlap_l_MSTR"!(ComponentPose=@n6, bAlphaBoolEnabled:bool=@n3, Translation="0,-1,0", Rotation="0,0,0", Scale="1,1,1") -> (Pose) {
    alphaInputType=Bool
    rotationMode=BMM_Additive
    rotationSpace=BCS_BoneSpace
    translationMode=BMM_Additive
    translationSpace=BCS_BoneSpace
    alphaBoolBlend {blendOption=Linear}
    boneToModify {boneName=dyn_scarfFlap_l_MSTR}
    translation {y=-1}
   }
   n8 = "Transform (Modify) Bone - Bone: dyn_bentoBagMain_MSTR"!(ComponentPose=@n7, bAlphaBoolEnabled:bool=Get bIsSkyDiving(), Translation="0,0,0", Rotation="-15,-20,0", Scale="1,1,1") -> (Pose) {
    alphaInputType=Bool
    rotationMode=BMM_Additive
    rotationSpace=BCS_BoneSpace
    translationMode=BMM_Additive
    translationSpace=BCS_BoneSpace
    alphaBoolBlend {blendOption=Linear}
    boneToModify {boneName=dyn_bentoBagMain_MSTR}
    rotation {pitch=-15, yaw=-20}
   }
   n9 = "Transform (Modify) Bone - Bone: dyn_shoulderPad_l_01"(ComponentPose=@n8, bAlphaBoolEnabled=@n2, Translation="9,0,-1", Rotation="0,0,10", Scale="1,1,1") {
    alphaInputType=Bool
    rotationMode=BMM_Additive
    rotationSpace=BCS_BoneSpace
    translationMode=BMM_Additive
    translationSpace=BCS_BoneSpace
    alphaBoolBlend {blendOption=Linear}
    boneToModify {boneName=dyn_shoulderPad_l_01}
    rotation {roll=10}
    translation {x=9, z=-1}
   }
   n10 = "Transform (Modify) Bone - Bone: dyn_shoulderPlateC_l_01"(ComponentPose=@n9, bAlphaBoolEnabled=@n2, Translation="-2,0,0", Rotation="0,0,0", Scale="1,1,1") {
    alphaInputType=Bool
    rotationMode=BMM_Additive
    rotationSpace=BCS_BoneSpace
    translationMode=BMM_Additive
    translationSpace=BCS_BoneSpace
    alphaBoolBlend {blendOption=Linear}
    boneToModify {boneName=dyn_shoulderPlateC_l_01}
    translation {x=-2}
   }
   n11 = "Transform (Modify) Bone - Bone: dyn_shoulderPlateB_l_01"(ComponentPose=@n10, bAlphaBoolEnabled=@n2, Translation="-2,0,0", Rotation="0,0,0", Scale="1,1,1") {
    alphaInputType=Bool
    rotationMode=BMM_Additive
    rotationSpace=BCS_BoneSpace
    translationMode=BMM_Additive
    translationSpace=BCS_BoneSpace
    alphaBoolBlend {blendOption=Linear}
    boneToModify {boneName=dyn_shoulderPlateB_l_01}
    translation {x=-2}
   }
   n12 = "Transform (Modify) Bone - Bone: dyn_scarfFlap_l_MSTR"(ComponentPose=@n11, bAlphaBoolEnabled=@n2, Translation="0,-1,0", Rotation="0,0,0", Scale="1,1,1") {
    alphaInputType=Bool
    rotationMode=BMM_Additive
    rotationSpace=BCS_BoneSpace
    translationMode=BMM_Additive
    translationSpace=BCS_BoneSpace
    alphaBoolBlend {blendOption=Linear}
    boneToModify {boneName=dyn_scarfFlap_l_MSTR}
    translation {y=-1}
   }
   n13 = "Transform (Modify) Bone - Bone: dyn_skirtA_l_MSTR"(ComponentPose=@n12, Alpha=@n1, Translation="0,0,0", Rotation="0,10,0", Scale="1,1,1") {
    rotationMode=BMM_Additive
    rotationSpace=BCS_BoneSpace
    alphaBoolBlend {blendOption=Linear}
    alphaScaleBiasClamp {bInterpResult=true, interpSpeedIncreasing=1}
    boneToModify {boneName=dyn_skirtA_l_MSTR}
    rotation {yaw=10}
   }
   n14 = "Transform (Modify) Bone - Bone: dyn_skirtA_r_MSTR"(ComponentPose=@n13, Alpha=@n1, Translation="0,0,0", Rotation="0,10,0", Scale="1,1,1") {
    rotationMode=BMM_Additive
    rotationSpace=BCS_BoneSpace
    alphaBoolBlend {blendOption=Linear}
    alphaScaleBiasClamp {bInterpResult=true, interpSpeedIncreasing=1}
    boneToModify {boneName=dyn_skirtA_r_MSTR}
    rotation {yaw=10}
   }
   n15 = "Transform (Modify) Bone - Bone: dyn_bowB_l_MSTR"(ComponentPose=@n14, Alpha=@n1, Translation="0,0,0", Rotation="0,6,0", Scale="1,1,1") {
    rotationMode=BMM_Additive
    rotationSpace=BCS_BoneSpace
    alphaBoolBlend {blendOption=Linear}
    alphaScaleBiasClamp {bInterpResult=true, interpSpeedIncreasing=1}
    boneToModify {boneName=dyn_bowB_l_MSTR}
    rotation {yaw=6}
   }
   Output Pose(Result=Component To Local(ComponentPose=@n15)) {layerGroup=None}
   n16 = RigidBody!(
    ComponentPose=Local To Component(LocalPose=Use cached pose 'rbanCache'())
    Alpha
    OverrideWorldGravity:Vector=Get Gravity Override()
    ExternalForce
   ) -> (Pose) {
    bClampLinearTranslationLimitToRefPose=true
    bOverrideWorldGravity=true
    overridePhysicsAsset=&/M_LRG_HealedScar_Physics
    simulationSpace=BaseBoneSpace
    alphaBoolBlend {blendOption=Linear}
    baseBoneRef {boneName=pelvis}
    componentAppliedLinearAccClamp=<3000 3000 3000>
    componentLinearAccScale=<0.3 0.3 0.3>
    componentLinearVelScale=<0.3 0.3 0.3>
    simSpaceSettings {maxAngularAcceleration=1, maxAngularVelocity=5, maxLinearAcceleration=500, maxLinearVelocity=1500, externalLinearDragV=<1 1>}
   }
   n17 = Anim Dynamics - Bone: dyn_skirtA_l_03!(
    SimulationSpace:AnimPhysSimSpaceType Enum
    ComponentPose=@n16
    Alpha=1
    ComponentLinearAccScale:Vector="1,1,1"
    ComponentLinearVelScale:Vector="1,1,1"
   ) -> (Pose) {
    angularDampingOverride=1
    angularSpringConstant=8000
    bAngularSpring=true
    bOverrideAngularDamping=true
    bUseGravityOverride=true
    gravityScale=0.35
    simulationSpace=RootRelative
    alphaBoolBlend {blendOption=Linear}
    boundBone {boneName=dyn_skirtA_l_03}
    componentAppliedLinearAccClamp=<3000 3000 3000>
    componentLinearAccScale=<1 1 1>
    componentLinearVelScale=<1 1 1>
    gravityOverride {z=-980}
    physicsBodyDefinitions [{
      boundBone {boneName=dyn_skirtA_l_03}
      boxExtents=<20 30 30>
      constraintSetup {angularLimitsMax=<0 3 5>, angularLimitsMin=<0 -3 -10>, angularTarget=<1 0 0>}
      localJointOffset=<4 0 0>
     }]
   }
   n18 = Anim Dynamics - Bone: dyn_skirtA_l_04!(
    SimulationSpace:AnimPhysSimSpaceType Enum
    ComponentPose=@n17
    Alpha=1
    ComponentLinearAccScale:Vector="1,1,1"
    ComponentLinearVelScale:Vector="1,1,1"
   ) -> (Pose) {
    angularDampingOverride=1
    angularSpringConstant=8000
    bAngularSpring=true
    bOverrideAngularDamping=true
    bUseGravityOverride=true
    gravityScale=0.35
    simulationSpace=RootRelative
    alphaBoolBlend {blendOption=Linear}
    boundBone {boneName=dyn_skirtA_l_04}
    componentAppliedLinearAccClamp=<3000 3000 3000>
    componentLinearAccScale=<1 1 1>
    componentLinearVelScale=<1 1 1>
    gravityOverride {z=-980}
    physicsBodyDefinitions [{
      boundBone {boneName=dyn_skirtA_l_04}
      boxExtents=<20 30 30>
      constraintSetup {angularLimitsMax=<0 3 5>, angularLimitsMin=<0 -3 -10>, angularTarget=<1 0 0>}
      localJointOffset=<4 0 0>
     }]
   }
   n19 = Anim Dynamics - Bone: dyn_skirtB_l_03!(
    SimulationSpace:AnimPhysSimSpaceType Enum
    ComponentPose=@n18
    Alpha=1
    ComponentLinearAccScale:Vector="1,1,1"
    ComponentLinearVelScale:Vector="1,1,1"
   ) -> (Pose) {
    angularDampingOverride=1
    angularSpringConstant=8000
    bAngularSpring=true
    bOverrideAngularDamping=true
    bUseGravityOverride=true
    gravityScale=0.35
    simulationSpace=RootRelative
    alphaBoolBlend {blendOption=Linear}
    boundBone {boneName=dyn_skirtB_l_03}
    componentAppliedLinearAccClamp=<3000 3000 3000>
    componentLinearAccScale=<1 1 1>
    componentLinearVelScale=<1 1 1>
    gravityOverride {z=-980}
    physicsBodyDefinitions [{
      boundBone {boneName=dyn_skirtB_l_03}
      boxExtents=<20 30 30>
      constraintSetup {angularLimitsMax=<0 3 5>, angularLimitsMin=<0 -3 -15>, angularTarget=<1 0 0>}
      localJointOffset=<4 0 0>
     }]
   }
   n20 = Anim Dynamics - Bone: dyn_skirtB_l_04!(
    SimulationSpace:AnimPhysSimSpaceType Enum
    ComponentPose=@n19
    Alpha=1
    ComponentLinearAccScale:Vector="1,1,1"
    ComponentLinearVelScale:Vector="1,1,1"
   ) -> (Pose) {
    angularDampingOverride=1
    angularSpringConstant=8000
    bAngularSpring=true
    bOverrideAngularDamping=true
    bUseGravityOverride=true
    gravityScale=0.35
    simulationSpace=RootRelative
    alphaBoolBlend {blendOption=Linear}
    boundBone {boneName=dyn_skirtB_l_04}
    componentAppliedLinearAccClamp=<3000 3000 3000>
    componentLinearAccScale=<1 1 1>
    componentLinearVelScale=<1 1 1>
    gravityOverride {z=-980}
    physicsBodyDefinitions [{
      boundBone {boneName=dyn_skirtB_l_04}
      boxExtents=<20 30 30>
      constraintSetup {angularLimitsMax=<0 3 5>, angularLimitsMin=<0 -3 -15>, angularTarget=<1 0 0>}
      localJointOffset=<4 0 0>
     }]
   }
   n21 = Anim Dynamics - Bone: dyn_skirtA_r_03!(
    SimulationSpace:AnimPhysSimSpaceType Enum
    ComponentPose=@n20
    Alpha=1
    ComponentLinearAccScale:Vector="1,1,1"
    ComponentLinearVelScale:Vector="1,1,1"
   ) -> (Pose) {
    angularDampingOverride=1
    angularSpringConstant=8000
    bAngularSpring=true
    bOverrideAngularDamping=true
    bUseGravityOverride=true
    gravityScale=0.35
    simulationSpace=RootRelative
    alphaBoolBlend {blendOption=Linear}
    boundBone {boneName=dyn_skirtA_r_03}
    componentAppliedLinearAccClamp=<3000 3000 3000>
    componentLinearAccScale=<1 1 1>
    componentLinearVelScale=<1 1 1>
    gravityOverride {z=-980}
    physicsBodyDefinitions [{
      boundBone {boneName=dyn_skirtA_r_03}
      boxExtents=<20 30 30>
      constraintSetup {angularLimitsMax=<0 3 5>, angularLimitsMin=<0 -3 -15>, angularTarget=<1 0 0>}
      localJointOffset=<4 0 0>
     }]
   }
   n22 = Anim Dynamics - Bone: dyn_skirtA_r_04!(
    SimulationSpace:AnimPhysSimSpaceType Enum
    ComponentPose=@n21
    Alpha=1
    ComponentLinearAccScale:Vector="1,1,1"
    ComponentLinearVelScale:Vector="1,1,1"
   ) -> (Pose) {
    angularDampingOverride=1
    angularSpringConstant=8000
    bAngularSpring=true
    bOverrideAngularDamping=true
    bUseGravityOverride=true
    gravityScale=0.35
    simulationSpace=RootRelative
    alphaBoolBlend {blendOption=Linear}
    boundBone {boneName=dyn_skirtA_r_04}
    componentAppliedLinearAccClamp=<3000 3000 3000>
    componentLinearAccScale=<1 1 1>
    componentLinearVelScale=<1 1 1>
    gravityOverride {z=-980}
    physicsBodyDefinitions [{
      boundBone {boneName=dyn_skirtA_r_04}
      boxExtents=<20 30 30>
      constraintSetup {angularLimitsMax=<0 3 5>, angularLimitsMin=<0 -3 -15>, angularTarget=<1 0 0>}
      localJointOffset=<4 0 0>
     }]
   }
   n23 = Anim Dynamics - Bone: dyn_skirtB_r_03!(
    SimulationSpace:AnimPhysSimSpaceType Enum
    ComponentPose=@n22
    Alpha=1
    ComponentLinearAccScale:Vector="1,1,1"
    ComponentLinearVelScale:Vector="1,1,1"
   ) -> (Pose) {
    angularDampingOverride=1
    angularSpringConstant=8000
    bAngularSpring=true
    bOverrideAngularDamping=true
    bUseGravityOverride=true
    gravityScale=0.35
    simulationSpace=RootRelative
    alphaBoolBlend {blendOption=Linear}
    boundBone {boneName=dyn_skirtB_r_03}
    componentAppliedLinearAccClamp=<3000 3000 3000>
    componentLinearAccScale=<1 1 1>
    componentLinearVelScale=<1 1 1>
    gravityOverride {z=-980}
    physicsBodyDefinitions [{
      boundBone {boneName=dyn_skirtB_r_03}
      boxExtents=<20 30 30>
      constraintSetup {angularLimitsMax=<0 3 5>, angularLimitsMin=<0 -3 -15>, angularTarget=<1 0 0>}
      localJointOffset=<4 0 0>
     }]
   }
   n24 = Anim Dynamics - Bone: dyn_skirtB_r_04!(
    SimulationSpace:AnimPhysSimSpaceType Enum
    ComponentPose=@n23
    Alpha=1
    ComponentLinearAccScale:Vector="1,1,1"
    ComponentLinearVelScale:Vector="1,1,1"
   ) -> (Pose) {
    angularDampingOverride=1
    angularSpringConstant=8000
    bAngularSpring=true
    bOverrideAngularDamping=true
    bUseGravityOverride=true
    gravityScale=0.35
    simulationSpace=RootRelative
    alphaBoolBlend {blendOption=Linear}
    boundBone {boneName=dyn_skirtB_r_04}
    componentAppliedLinearAccClamp=<3000 3000 3000>
    componentLinearAccScale=<1 1 1>
    componentLinearVelScale=<1 1 1>
    gravityOverride {z=-980}
    physicsBodyDefinitions [{
      boundBone {boneName=dyn_skirtB_r_04}
      boxExtents=<20 30 30>
      constraintSetup {angularLimitsMax=<0 3 5>, angularLimitsMin=<0 -3 -15>, angularTarget=<1 0 0>}
      localJointOffset=<4 0 0>
     }]
   }
   Save cached pose 'SkirtCache'(Pose=Component To Local(ComponentPose=@n24)) {cacheName=SkirtCache}
   n25 = Get isNotWearingBackbling() -> (isNotWearingBackbling:bool)
   n26 = Anim Dynamics - Bone: dyn_bentoBagMain_Pivot(ComponentPose=Local To Component(LocalPose=Use cached pose 'preSimCache'()), Alpha=1, ExternalForce="0,0,0") {
    angularDampingOverride=1
    angularSpringConstant=5000
    bAngularSpring=true
    bOverrideAngularDamping=true
    bUseGravityOverride=true
    gravityScale=0.35
    simulationSpace=RootRelative
    alphaBoolBlend {blendOption=Linear}
    boundBone {boneName=dyn_bentoBagMain_Pivot}
    componentAppliedLinearAccClamp=<3000 3000 3000>
    componentLinearAccScale=<1 1 1>
    componentLinearVelScale=<1 1 1>
    gravityOverride {z=-980}
    physicsBodyDefinitions [{
      boundBone {boneName=dyn_bentoBagMain_Pivot}
      boxExtents=<20 30 30>
      constraintSetup {angularLimitsMax=<0 0 -10>, angularLimitsMin=<0 0 -30>, angularTarget=<1 0 0>}
      localJointOffset=<4 0 0>
     }]
   }
   n27 = Anim Dynamics - Bone: dyn_bentoBagMain_01(ComponentPose=@n26, Alpha=1, ExternalForce="0,0,0") {
    angularDampingOverride=1
    angularSpringConstant=10000
    bAngularSpring=true
    bOverrideAngularDamping=true
    bUseGravityOverride=true
    gravityScale=0.35
    simulationSpace=RootRelative
    alphaBoolBlend {blendOption=Linear}
    boundBone {boneName=dyn_bentoBagMain_01}
    componentAppliedLinearAccClamp=<3000 3000 3000>
    componentLinearAccScale=<1 1 1>
    componentLinearVelScale=<1 1 1>
    gravityOverride {z=-980}
    physicsBodyDefinitions [{
      boundBone {boneName=dyn_bentoBagMain_01}
      boxExtents=<20 30 30>
      constraintSetup {angularLimitsMax=<15 30 20>, angularLimitsMin=<-15 0 -20>, angularTarget=<1 0 0>}
      localJointOffset=<4 0 0>
     }]
   }
   n28 = Anim Dynamics - Bone: dyn_bentoBagF_01(ComponentPose=@n27, Alpha=1, ExternalForce="0,0,0") {
    angularDampingOverride=1
    angularSpringConstant=12000
    bAngularSpring=true
    bOverrideAngularDamping=true
    bUseGravityOverride=true
    gravityScale=0.35
    simulationSpace=RootRelative
    alphaBoolBlend {blendOption=Linear}
    boundBone {boneName=dyn_bentoBagF_01}
    componentAppliedLinearAccClamp=<3000 3000 3000>
    componentLinearAccScale=<1 1 1>
    componentLinearVelScale=<1 1 1>
    gravityOverride {z=-980}
    physicsBodyDefinitions [{
      boundBone {boneName=dyn_bentoBagF_01}
      boxExtents=<20 30 30>
      constraintSetup {angularLimitsMax=<15 15 15>, angularLimitsMin=<-15 5 -15>, angularTarget=<1 0 0>}
      localJointOffset=<4 0 0>
     }]
   }
   n29 = Anim Dynamics - Bone: dyn_bentoBagE_01(ComponentPose=@n28, Alpha=1, ExternalForce="0,0,0") {
    angularDampingOverride=1
    angularSpringConstant=12000
    bAngularSpring=true
    bOverrideAngularDamping=true
    bUseGravityOverride=true
    gravityScale=0.35
    simulationSpace=RootRelative
    alphaBoolBlend {blendOption=Linear}
    boundBone {boneName=dyn_bentoBagE_01}
    componentAppliedLinearAccClamp=<3000 3000 3000>
    componentLinearAccScale=<1 1 1>
    componentLinearVelScale=<1 1 1>
    gravityOverride {z=-980}
    physicsBodyDefinitions [{
      boundBone {boneName=dyn_bentoBagE_01}
      boxExtents=<20 30 30>
      constraintSetup {angularLimitsMax=<15 15 15>, angularLimitsMin=<-15 -15 -15>, angularTarget=<1 0 0>}
      localJointOffset=<4 0 0>
     }]
   }
   n30 = Anim Dynamics - Bone: dyn_bentoBagD_01(ComponentPose=@n29, Alpha=1, ExternalForce="0,0,0") {
    angularDampingOverride=1
    angularSpringConstant=12000
    bAngularSpring=true
    bOverrideAngularDamping=true
    bUseGravityOverride=true
    gravityScale=0.35
    simulationSpace=RootRelative
    alphaBoolBlend {blendOption=Linear}
    boundBone {boneName=dyn_bentoBagD_01}
    componentAppliedLinearAccClamp=<3000 3000 3000>
    componentLinearAccScale=<1 1 1>
    componentLinearVelScale=<1 1 1>
    gravityOverride {z=-980}
    physicsBodyDefinitions [{
      boundBone {boneName=dyn_bentoBagD_01}
      boxExtents=<20 30 30>
      constraintSetup {angularLimitsMax=<15 15 15>, angularLimitsMin=<-15 -15 -15>, angularTarget=<1 0 0>}
      localJointOffset=<4 0 0>
     }]
   }
   n31 = Anim Dynamics - Bone: dyn_bentoBagC_01(ComponentPose=@n30, Alpha=1, ExternalForce="0,0,0") {
    angularDampingOverride=1
    angularSpringConstant=12000
    bAngularSpring=true
    bOverrideAngularDamping=true
    bUseGravityOverride=true
    gravityScale=0.35
    simulationSpace=RootRelative
    alphaBoolBlend {blendOption=Linear}
    boundBone {boneName=dyn_bentoBagC_01}
    componentAppliedLinearAccClamp=<3000 3000 3000>
    componentLinearAccScale=<1 1 1>
    componentLinearVelScale=<1 1 1>
    gravityOverride {z=-980}
    physicsBodyDefinitions [{
      boundBone {boneName=dyn_bentoBagC_01}
      boxExtents=<20 30 30>
      constraintSetup {angularLimitsMax=<15 15 15>, angularLimitsMin=<-15 -15 -15>, angularTarget=<1 0 0>}
      localJointOffset=<4 0 0>
     }]
   }
   n32 = Anim Dynamics - Bone: dyn_bentoBagB_01(ComponentPose=@n31, Alpha=1, ExternalForce="0,0,0") {
    angularDampingOverride=1
    angularSpringConstant=12000
    bAngularSpring=true
    bOverrideAngularDamping=true
    bUseGravityOverride=true
    gravityScale=0.35
    simulationSpace=RootRelative
    alphaBoolBlend {blendOption=Linear}
    boundBone {boneName=dyn_bentoBagB_01}
    componentAppliedLinearAccClamp=<3000 3000 3000>
    componentLinearAccScale=<1 1 1>
    componentLinearVelScale=<1 1 1>
    gravityOverride {z=-980}
    physicsBodyDefinitions [{
      boundBone {boneName=dyn_bentoBagB_01}
      boxExtents=<20 30 30>
      constraintSetup {angularLimitsMax=<15 15 15>, angularLimitsMin=<-15 -15 -15>, angularTarget=<1 0 0>}
      localJointOffset=<4 0 0>
     }]
   }
   n33 = Anim Dynamics - Bone: dyn_bentoBagA_01(ComponentPose=@n32, Alpha=1, ExternalForce="0,0,0") {
    angularDampingOverride=1
    angularSpringConstant=12000
    bAngularSpring=true
    bOverrideAngularDamping=true
    bUseGravityOverride=true
    gravityScale=0.35
    simulationSpace=RootRelative
    alphaBoolBlend {blendOption=Linear}
    boundBone {boneName=dyn_bentoBagA_01}
    componentAppliedLinearAccClamp=<3000 3000 3000>
    componentLinearAccScale=<1 1 1>
    componentLinearVelScale=<1 1 1>
    gravityOverride {z=-980}
    physicsBodyDefinitions [{
      boundBone {boneName=dyn_bentoBagA_01}
      boxExtents=<20 30 30>
      constraintSetup {angularLimitsMax=<15 15 15>, angularLimitsMin=<-15 -15 -15>, angularTarget=<1 0 0>}
      localJointOffset=<4 0 0>
     }]
   }
   n34 = Look At - Bone: dyn_bentoRopeAim_01(ComponentPose=@n33, Alpha=1) {
    alphaBoolBlend {blendOption=Linear}
    boneToModify {boneName=dyn_bentoRopeAim_01}
    lookAtLocation=<10 50>
    lookAtTarget {boneReference {boneName=dyn_bentoBagMain_MSTR}}
    lookAt_Axis {axis {x=1}}
    lookUp_Axis {axis {x=0, z=1}}
   }
   n35 = Anim Dynamics - Bone: dyn_scarfFlap_l_03!(ComponentPose=@n34, bAlphaBoolEnabled:bool=@n25, GravityOverride:Vector="0,0,-980", ExternalForce="0,0,0") -> (Pose) {
    alphaInputType=Bool
    angularDampingOverride=1
    angularSpringConstant=2000
    bAngularSpring=true
    bOverrideAngularDamping=true
    bUseGravityOverride=true
    gravityScale=0.35
    simulationSpace=BoneRelative
    alphaBoolBlend {blendOption=Linear}
    boundBone {boneName=dyn_scarfFlap_l_03}
    componentAppliedLinearAccClamp=<2000 2000 2000>
    componentLinearAccScale {y=0.4, z=0.4}
    componentLinearVelScale {y=0.4, z=0.4}
    gravityOverride {z=-980}
    physicsBodyDefinitions [{
      boundBone {boneName=dyn_scarfFlap_l_03}
      boxExtents=<20 30 30>
      constraintSetup {angularLimitsMax=<0 0 0>, angularLimitsMin=<0 0 -30>, angularTarget=<1 0 0>}
      localJointOffset=<4 0 0>
     }]
    relativeSpaceBone {boneName=root}
   }
   n36 = Anim Dynamics - Bone: dyn_scarfFlap_l_02!(ComponentPose=@n35, bAlphaBoolEnabled:bool=@n25, GravityOverride:Vector="0,0,-980", ExternalForce="0,0,0") -> (Pose) {
    alphaInputType=Bool
    angularDampingOverride=1
    angularSpringConstant=2000
    bAngularSpring=true
    bOverrideAngularDamping=true
    bUseGravityOverride=true
    gravityScale=0.35
    simulationSpace=BoneRelative
    alphaBoolBlend {blendOption=Linear}
    boundBone {boneName=dyn_scarfFlap_l_02}
    componentAppliedLinearAccClamp=<2000 2000 2000>
    componentLinearAccScale {y=0.4, z=0.4}
    componentLinearVelScale {y=0.4, z=0.4}
    gravityOverride {z=-980}
    physicsBodyDefinitions [{
      boundBone {boneName=dyn_scarfFlap_l_02}
      boxExtents=<20 30 30>
      constraintSetup {angularLimitsMax=<0 0 0>, angularLimitsMin=<0 0 -30>, angularTarget=<1 0 0>}
      localJointOffset=<4 0 0>
     }]
    relativeSpaceBone {boneName=root}
   }
   n37 = Anim Dynamics - Bone: dyn_scarfFlap_l_01!(ComponentPose=@n36, bAlphaBoolEnabled:bool=@n25, GravityOverride:Vector="0,0,-980", ExternalForce="0,0,0") -> (Pose) {
    alphaInputType=Bool
    angularDampingOverride=1
    angularSpringConstant=2000
    bAngularSpring=true
    bOverrideAngularDamping=true
    bUseGravityOverride=true
    gravityScale=0.35
    simulationSpace=BoneRelative
    alphaBoolBlend {blendOption=Linear}
    boundBone {boneName=dyn_scarfFlap_l_01}
    componentAppliedLinearAccClamp=<2000 2000 2000>
    componentLinearAccScale {y=0.4, z=0.4}
    componentLinearVelScale {y=0.4, z=0.4}
    gravityOverride {z=-980}
    physicsBodyDefinitions [{
      boundBone {boneName=dyn_scarfFlap_l_01}
      boxExtents=<20 30 30>
      constraintSetup {angularLimitsMax=<0 0 0>, angularLimitsMin=<0 0 -30>, angularTarget=<1 0 0>}
      localJointOffset=<4 0 0>
     }]
    relativeSpaceBone {boneName=root}
   }
   n38 = Anim Dynamics - Bone: dyn_fluffB_r_01(ComponentPose=@n37, Alpha=1, ExternalForce="0,0,0") {
    angularDampingOverride=1
    angularSpringConstant=15000
    bAngularSpring=true
    bOverrideAngularDamping=true
    bUseGravityOverride=true
    gravityScale=0.35
    simulationSpace=RootRelative
    alphaBoolBlend {blendOption=Linear}
    boundBone {boneName=dyn_fluffB_r_01}
    componentAppliedLinearAccClamp=<3000 3000 3000>
    componentLinearAccScale=<1 1 1>
    componentLinearVelScale=<1 1 1>
    gravityOverride {z=-980}
    physicsBodyDefinitions [{
      boundBone {boneName=dyn_fluffB_r_01}
      boxExtents=<20 30 30>
      constraintSetup {angularLimitsMax=<0 5 0>, angularLimitsMin=<0 -5 0>, angularTarget=<1 0 0>}
      localJointOffset=<4 0 0>
     }]
   }
   n39 = Anim Dynamics - Bone: dyn_fluffA_r_01(ComponentPose=@n38, Alpha=1, ExternalForce="0,0,0") {
    angularDampingOverride=1
    angularSpringConstant=15000
    bAngularSpring=true
    bOverrideAngularDamping=true
    bUseGravityOverride=true
    gravityScale=0.35
    simulationSpace=RootRelative
    alphaBoolBlend {blendOption=Linear}
    boundBone {boneName=dyn_fluffA_r_01}
    componentAppliedLinearAccClamp=<3000 3000 3000>
    componentLinearAccScale=<1 1 1>
    componentLinearVelScale=<1 1 1>
    gravityOverride {z=-980}
    physicsBodyDefinitions [{
      boundBone {boneName=dyn_fluffA_r_01}
      boxExtents=<20 30 30>
      constraintSetup {angularLimitsMax=<0 5 0>, angularLimitsMin=<0 -5 0>, angularTarget=<1 0 0>}
      localJointOffset=<4 0 0>
     }]
   }
   n40 = Anim Dynamics - Bone: dyn_scarf_c_01(ComponentPose=@n39, Alpha=1, ExternalForce="0,0,0") {
    angularDampingOverride=1
    angularSpringConstant=20000
    bAngularSpring=true
    bOverrideAngularDamping=true
    bUseGravityOverride=true
    gravityScale=0.35
    simulationSpace=RootRelative
    alphaBoolBlend {blendOption=Linear}
    boundBone {boneName=dyn_scarf_c_01}
    componentAppliedLinearAccClamp=<3000 3000 3000>
    componentLinearAccScale=<1 1 1>
    componentLinearVelScale=<1 1 1>
    gravityOverride {z=-980}
    physicsBodyDefinitions [{
      boundBone {boneName=dyn_scarf_c_01}
      boxExtents=<20 30 30>
      constraintSetup {angularLimitsMax=<0 3 0>, angularLimitsMin=<0 -3 -15>, angularTarget=<1 0 0>}
      localJointOffset=<4 0 0>
     }]
   }
   n41 = Anim Dynamics - Bone: dyn_shirtJiggle_c_01(ComponentPose=@n40, Alpha=1, ExternalForce="0,0,0") {
    angularDampingOverride=1
    angularSpringConstant=20000
    bLinearSpring=true
    bOverrideAngularDamping=true
    bUseGravityOverride=true
    bUsePlanarLimit=false
    gravityScale=0.35
    linearSpringConstant=0.9999
    alphaBoolBlend {blendOption=Linear}
    boundBone {boneName=dyn_shirtJiggle_c_01}
    componentAppliedLinearAccClamp=<3000 3000 3000>
    componentLinearAccScale=<0.35 0.4 0.4>
    componentLinearVelScale=<0.35 0.4 0.4>
    gravityOverride {z=-980}
    physicsBodyDefinitions [{
      boundBone {boneName=dyn_shirtJiggle_c_01}
      boxExtents=<20 30 30>
      constraintSetup {angularLimitsMax=<0 0 0>, angularLimitsMin=<0 0 0>, angularTarget=<1 0 0>, linearAxesMax=<3 0 1>, linearAxesMin=<-1.5 -2 -1>}
      localJointOffset=<4 0 0>
     }]
    relativeSpaceBone {boneName=root}
   }
   n42 = Anim Dynamics - Bone: dyn_legFluffC_l_01(ComponentPose=@n41, Alpha=1, ExternalForce="0,0,0") {
    angularDampingOverride=1
    angularSpringConstant=10000
    bAngularSpring=true
    bOverrideAngularDamping=true
    bUseGravityOverride=true
    gravityScale=0.35
    simulationSpace=RootRelative
    alphaBoolBlend {blendOption=Linear}
    boundBone {boneName=dyn_legFluffC_l_01}
    componentAppliedLinearAccClamp=<3000 3000 3000>
    componentLinearAccScale=<1 1 1>
    componentLinearVelScale=<1 1 1>
    gravityOverride {z=-980}
    physicsBodyDefinitions [{
      boundBone {boneName=dyn_legFluffC_l_01}
      boxExtents=<20 30 30>
      constraintSetup {angularLimitsMax=<15 15 15>, angularLimitsMin=<-15 -15 -15>, angularTarget=<1 0 0>}
      localJointOffset=<4 0 0>
     }]
   }
   n43 = Anim Dynamics - Bone: dyn_legFluffB_l_01(ComponentPose=@n42, Alpha=1, ExternalForce="0,0,0") {
    angularDampingOverride=1
    angularSpringConstant=14000
    bAngularSpring=true
    bOverrideAngularDamping=true
    bUseGravityOverride=true
    gravityScale=0.35
    simulationSpace=RootRelative
    alphaBoolBlend {blendOption=Linear}
    boundBone {boneName=dyn_legFluffB_l_01}
    componentAppliedLinearAccClamp=<3000 3000 3000>
    componentLinearAccScale=<1 1 1>
    componentLinearVelScale=<1 1 1>
    gravityOverride {z=-980}
    physicsBodyDefinitions [{
      boundBone {boneName=dyn_legFluffB_l_01}
      boxExtents=<20 30 30>
      constraintSetup {angularLimitsMax=<15 15 15>, angularLimitsMin=<-15 -15 -15>, angularTarget=<1 0 0>}
      localJointOffset=<4 0 0>
     }]
   }
   n44 = Anim Dynamics - Bone: dyn_legFluffA_l_01(ComponentPose=@n43, Alpha=1, ExternalForce="0,0,0") {
    angularDampingOverride=1
    angularSpringConstant=20000
    bAngularSpring=true
    bOverrideAngularDamping=true
    bUseGravityOverride=true
    gravityScale=0.35
    simulationSpace=RootRelative
    alphaBoolBlend {blendOption=Linear}
    boundBone {boneName=dyn_legFluffA_l_01}
    componentAppliedLinearAccClamp=<3000 3000 3000>
    componentLinearAccScale=<1 1 1>
    componentLinearVelScale=<1 1 1>
    gravityOverride {z=-980}
    physicsBodyDefinitions [{
      boundBone {boneName=dyn_legFluffA_l_01}
      boxExtents=<20 30 30>
      constraintSetup {angularLimitsMax=<15 15 15>, angularLimitsMin=<-15 -15 -15>, angularTarget=<1 0 0>}
      localJointOffset=<4 0 0>
     }]
   }
   n45 = Anim Dynamics - Bone: dyn_legFluffA_r_01(ComponentPose=@n44, Alpha=1, ExternalForce="0,0,0") {
    angularDampingOverride=1
    angularSpringConstant=20000
    bAngularSpring=true
    bOverrideAngularDamping=true
    bUseGravityOverride=true
    gravityScale=0.35
    simulationSpace=RootRelative
    alphaBoolBlend {blendOption=Linear}
    boundBone {boneName=dyn_legFluffA_r_01}
    componentAppliedLinearAccClamp=<3000 3000 3000>
    componentLinearAccScale=<1 1 1>
    componentLinearVelScale=<1 1 1>
    gravityOverride {z=-980}
    physicsBodyDefinitions [{
      boundBone {boneName=dyn_legFluffA_r_01}
      boxExtents=<20 30 30>
      constraintSetup {angularLimitsMax=<15 15 15>, angularLimitsMin=<-15 -15 -15>, angularTarget=<1 0 0>}
      localJointOffset=<4 0 0>
     }]
   }
   n46 = Anim Dynamics - Bone: dyn_legFluffC_r_01(ComponentPose=@n45, Alpha=1, ExternalForce="0,0,0") {
    angularDampingOverride=1
    angularSpringConstant=10000
    bAngularSpring=true
    bOverrideAngularDamping=true
    bUseGravityOverride=true
    gravityScale=0.35
    simulationSpace=RootRelative
    alphaBoolBlend {blendOption=Linear}
    boundBone {boneName=dyn_legFluffC_r_01}
    componentAppliedLinearAccClamp=<3000 3000 3000>
    componentLinearAccScale=<1 1 1>
    componentLinearVelScale=<1 1 1>
    gravityOverride {z=-980}
    physicsBodyDefinitions [{
      boundBone {boneName=dyn_legFluffC_r_01}
      boxExtents=<20 30 30>
      constraintSetup {angularLimitsMax=<15 15 15>, angularLimitsMin=<-15 -15 -15>, angularTarget=<1 0 0>}
      localJointOffset=<4 0 0>
     }]
   }
   n47 = Anim Dynamics - Bone: dyn_legFluffB_r_01(ComponentPose=@n46, Alpha=1, ExternalForce="0,0,0") {
    angularDampingOverride=1
    angularSpringConstant=14000
    bAngularSpring=true
    bOverrideAngularDamping=true
    bUseGravityOverride=true
    gravityScale=0.35
    simulationSpace=RootRelative
    alphaBoolBlend {blendOption=Linear}
    boundBone {boneName=dyn_legFluffB_r_01}
    componentAppliedLinearAccClamp=<3000 3000 3000>
    componentLinearAccScale=<1 1 1>
    componentLinearVelScale=<1 1 1>
    gravityOverride {z=-980}
    physicsBodyDefinitions [{
      boundBone {boneName=dyn_legFluffB_r_01}
      boxExtents=<20 30 30>
      constraintSetup {angularLimitsMax=<15 15 15>, angularLimitsMin=<-15 -15 -15>, angularTarget=<1 0 0>}
      localJointOffset=<4 0 0>
     }]
   }
   Save cached pose 'rbanCache'(Pose=Component To Local(ComponentPose=@n47)) {cacheName=rbanCache}
   n48 = RigidBody(ComponentPose=Local To Component(LocalPose=Use cached pose 'SkirtCache'()), OverrideWorldGravity=Get Gravity Override()) {
    bClampLinearTranslationLimitToRefPose=true
    bOverrideWorldGravity=true
    overridePhysicsAsset=&/M_LRG_HealedScar_Bow_isCrouching_Physics
    simulationSpace=BaseBoneSpace
    alphaBoolBlend {blendOption=Linear}
    baseBoneRef {boneName=pelvis}
    componentAppliedLinearAccClamp=<6000 6000 6000>
    componentLinearAccScale=<0.3 0.3 0.3>
    componentLinearVelScale=<0.3 0.3 0.3>
    simSpaceSettings {maxAngularAcceleration=1, maxAngularVelocity=5, maxLinearAcceleration=500, maxLinearVelocity=1500, worldAlpha=1, externalLinearDragV=<1 1>}
   }
   n49 = RigidBody(ComponentPose=Local To Component(LocalPose=Use cached pose 'SkirtCache'()), OverrideWorldGravity=Get Gravity Override()) {
    bClampLinearTranslationLimitToRefPose=true
    bOverrideWorldGravity=true
    overridePhysicsAsset=&M_LRG_HealedScar_Bow_isCrouching_Physics
    simulationSpace=BaseBoneSpace
    alphaBoolBlend {blendOption=Linear}
    baseBoneRef {boneName=pelvis}
    componentAppliedLinearAccClamp=<6000 6000 6000>
    componentLinearAccScale=<0.3 0.3 0.3>
    componentLinearVelScale=<0.3 0.3 0.3>
    simSpaceSettings {maxAngularAcceleration=1, maxAngularVelocity=5, maxLinearAcceleration=500, maxLinearVelocity=1500, worldAlpha=1, externalLinearDragV=<1 1>}
   }
   Save cached pose 'postSimCache'(Pose=Blend Poses by bool(BlendPose_0=Component To Local(ComponentPose=@n49), BlendPose_1=Component To Local(ComponentPose=@n48), bActiveValue=Get bIsCrouching()) {blendType=Linear}) {cacheName=postSimCache}
   n50 = Control Rig!(
    Source=Copy Pose From Mesh() {bUseAttachedParent=true}
    Alpha
    is_not_skydiving:bool=Get isNotSkydiving() -> (isNotSkydiving:bool)
    is_not_crouching:bool=NOT Boolean(A=Get bIsCrouching())
    is_player_moving:double=Get isPlayerMoving() -> (isPlayerMoving:double)
   ) -> (Pose) {
    alphaBoolBlend {blendOption=Linear}
    controlRigAssetReference {blueprintRigClass=&/M_LRG_HealedScar_CtrlRig.M_LRG_HealedScar_CtrlRig_C}
   }
   Save cached pose 'CRCache'(Pose=@n50) {cacheName=CRCache}
   Save cached pose 'preSimCache'(Pose=Use cached pose 'CRCache'()) {cacheName=preSimCache}
  ]}
 EventGraph {nodes [
   n1 = Event BlueprintUpdateAnimation()
   > Set Delta Time X(Delta Time X=@n1.DeltaTimeX)
   > n2 = Sequence!(execute) -> (then_0, then_1, then_2:Exec, then_3:Exec, then_4:Exec)
   n3 = Get bIsInFrontEnd()
   n4 = Branch(execute=@n2.then_2, Condition=@n3)
   > Set simSpaceChange(simSpaceChange:AnimPhysSimSpaceType Enum=World) -> (Output_Get:AnimPhysSimSpaceType Enum)
   > Set FELinearAccScale(FELinearAccScale:Vector="0, 0, 0") -> (Output_Get:Vector)
   Set simSpaceChange(execute=@n4.else)
   > Set FELinearAccScale(FELinearAccScale="1,1,1")
   n5 = PlayerMovement(execute=@n2.then_3, AnimInstance=Self-Reference(), XDotProductCoefficient=1, YDotProductCoefficient=1, ZDotProductCoefficient=1)
   > n6 = Set isPlayerMovingBol(isPlayerMovingBol:bool=OR Boolean!(A="Not Equal (Float)"(A=@n5), B="Not Equal (Float)"(A=@n5.YMovement), C:bool="Not Equal (Float)"(A=@n5.ZMovement)) -> (ReturnValue)) -> (Output_Get:bool)
   n7 = "To Float (Boolean)"(InBool=@n6)
   n8 = Get Delta Time X()
   > Set isPlayerMoving(isPlayerMoving:double=Remap Float(Input=FInterpTo(Current=@n7, Target=1, DeltaTime=@n8, InterpSpeed=0.1), OldRange=(X=0,Y=1), NewRange=(X=0.3,Y=1)) -> (Output=~)) -> (Output_Get:double)
   n9 = Get bIsPlayingEmote()
   n10 = Branch(execute=@n2.then_4, Condition=AND Boolean(A=Get bIsInFrontEnd(), B=NOT Boolean(A=@n9)))
   > Set Custom Idle is Playing Float(Custom Idle is Playing Float:float=1) -> (Output_Get:float)
   Set Custom Idle is Playing Float(execute=@n10.else, Custom Idle is Playing Float=0)
   n11 = Gravity Override State Function DEPRECATED(
    execute=@n2.then_0
    CharacterPartAnimInstance=Self-Reference()
    GameTimeInSeconds=Get Delta Time X()
    GravityOverrideParams=Get Gravity Override Params()
   )
   > Set Gravity Override(Gravity Override=@n11)
   n12 = Branch(execute=@n2.then_1, Condition=Get bIsSkyDiving())
   > Set isNotSkydiving(isNotSkydiving:bool=false) -> (Output_Get:bool)
   Set isNotSkydiving(execute=@n12.else, isNotSkydiving=true)
   Event OnParachuteOpened()
   > Set has_hand_glider(has_hand_glider:bool=false) -> (Output_Get:bool)
   > Switch on EGliderType(Selection=Get PlayerGliderType() -> (PlayerGliderType=HangGlider))
   > n13 = Set has_hand_glider(has_hand_glider=true)
   > Set has_hand_glider_and_is_skydiving(has_hand_glider_and_is_skydiving:bool=AND Boolean(A=@n13, B=Get bIsParachuteOpened())) -> (Output_Get:bool)
   Event OnParachuteClosed()
   > Set has_hand_glider(has_hand_glider=false)
   > Set has_hand_glider_and_is_skydiving(has_hand_glider_and_is_skydiving=false)
   Event OnCharacterAndPickaxeCustomizationLoaded!() -> (
    OutputDelegate
    then
    CustomBaseMesh:Skeletal Mesh Component Object Reference
    CustomParts:Array of Custom Character Part Object References
    CustomMeshParts:Array of Skeletal Mesh Component Budgeted Object References
   ) {is=title K2Node_Event}
   > Do Once()
   n14 = CheckForBackbling()
   > n15 = Branch(Condition=OR Boolean(A=@n14, B=@n14.bIsBackblingACape, C=@n14.bIsBackblingLong))
   > Set isNotWearingBackbling(isNotWearingBackbling:bool=false) -> (Output_Get:bool)
   Set isNotWearingBackbling(execute=@n15.else, isNotWearingBackbling=true)
   NOT Boolean(A=@n3)
  ]}
]
variables [
 Custom Idle is Playing Float {type=number}
 Delta Time X {type=number}
 FELinearAccScale {type=Vector}
 fort_player {type=FortPlayerAnimInstance, default=null}
 Gravity Override {type=Vector}
 Gravity Override Params {
  type=GravityOverrideParamsStruct
  default {skydive_DivingDown {gravityOverride=<-1200 0 0>}, skydive_GoingUp {gravityOverride=<0 -300 1200>}}
 }
 has_hand_glider {type=boolean}
 has_hand_glider_and_is_skydiving {type=boolean}
 isNotSkydiving {type=boolean}
 isNotWearingBackbling {type=boolean}
 isPlayerMoving {type=number}
 isPlayerMovingBol {type=boolean}
 NewVar {type=FortPlayerAnimInstance, default=null}
 simSpaceChange {type=AnimPhysSimSpaceType}
 Skel_Mesh {type=SkeletalMeshComponent}
]
